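# Translucent unlit deferred decal material for a cactus fruit Slurp splat, with a watery two-color look.
# Blended water normals drive the color lerp and refract scene-depth world positions into a tinted caustics function.
# Rotated drop and splash textures, cut by a blast-shaped mask and sphere mask, form opacity that erodes with FadeIn and decal lifetime.
# Shading path switches replace decal sphere-coordinate sampling with plain-UV variants on Forward and Mobile.
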Material /CactusFruits/VFX/MD_Decal_Cactus_Slurp
graph {
 properties {
  bAutomaticallySetUsageInEditor=false
  blendMode=BLEND_Translucent
  materialDomain=MD_DeferredDecal
  pixelDepthOffsetMode=PDOM_Legacy
  shadingModel=MSM_Unlit
  usedShadingModels=Unlit
 }
 nodes [
  Constant() {r=0.5}
  n1 = ComponentMask(None=VectorParameter() {group=Normals, parameterName=NormalUVScale, defaultValue=<rgba 1 0.381 0 0>}.RGB) {g=true, r=true}
  n2 = MaterialFunctionCall!(UVs=TextureCoordinate(), Additional Scale) {materialFunction=/Game/Effects/Fort_Effects/Materials/_Functions/MF_DecalSphereCoords}
  n3 = TextureSample(UVs=Add(A=Multiply(A=@n2.Result, B=@n1) {constB=0.8556}, B=ScalarParameter() {defaultValue=0.2, parameterName=B_1}) {constB=0.3215}) {samplerType=SAMPLERTYPE_Normal, texture=/Game/Global/Textures/Default/Tileable_Patterns/T_Water2_N}
  n4 = Multiply(A=ComponentMask(None=@n3.RGB) {g=true, r=true}, B=ScalarParameter() {defaultValue=0.1, group=Normals, parameterName=Noise 1}) {constB=0.04}
  n5 = Multiply(
   A=@n2.Result
   B=ComponentMask(None=VectorParameter() {group=Normals, parameterName=Normal2UVScale, defaultValue=<rgba 1 0.4667 0 0>}.RGB) {g=true, r=true}
  )
  n6 = MaterialFunctionCall(
   BaseNormal=@n3.RGB
   AdditionalNormal=TextureSample(UVs=Add(A=@n4, B=@n5)) {samplerType=SAMPLERTYPE_Normal, texture=&T_Water2_N}.RGB
  ) {materialFunction=&BlendAngleCorrectedNormals}
  n7 = TextureCoordinate()
  n8 = VectorParameter() {group=Normals_Mobile, parameterName=NormalUVScale_M3, sortPriority=1, defaultValue=<rgba 1 0.4667 0 0>}
  n9 = ComponentMask(None=@n8.RGB) {g=true, r=true}
  n10 = TextureSample!(UVs=VertexInterpolator(VS=Multiply(A=@n7, B=@n9)).PS, Tex, Level, Apply View MipBias) {mipValueMode=TMVM_MipLevel, samplerType=SAMPLERTYPE_Normal, texture=/Game/Global/Textures/Default/Tileable_Patterns/T_WaterSoft_N}
  n11 = Multiply(
   A=TextureCoordinate()
   B=ComponentMask(None=VectorParameter() {group=Normals_Mobile, parameterName=NormalUVScale_M/M2, defaultValue=<rgba 1 0.381 0 0>}.RGB) {g=true, r=true}
  ) {constB=0.8556}
  n12 = TextureSample(UVs=VertexInterpolator(VS=@n11).PS) {mipValueMode=TMVM_MipLevel, samplerType=SAMPLERTYPE_Normal, texture=&T_Water2_N}
  n13 = TextureSample(UVs=VertexInterpolator(VS=@n11).PS) {mipValueMode=TMVM_MipLevel, samplerType=SAMPLERTYPE_Normal, texture=&T_Water2_N}
  n14 = MaterialFunctionCall(
   BaseNormal=MaterialFunctionCall(Normal=@n10.RGB, Flatness=ScalarParameter() {defaultValue=0.9115, group=Normals_Mobile, parameterName=NormalFlaten_M3, sortPriority=5}) {materialFunction=&FlattenNormal}.Result
   AdditionalNormal=MaterialFunctionCall(
    Normal=MaterialFunctionCall(BaseNormal=@n12.RGB, AdditionalNormal=@n13.RGB) {materialFunction=&BlendAngleCorrectedNormals}.Result
    Flatness=ScalarParameter() {group=Normals, parameterName=NormalFlatten}
   ) {materialFunction=&FlattenNormal}.Result
  ) {materialFunction=&BlendAngleCorrectedNormals}
  n15 = ShadingPathSwitch(
   Default=MaterialFunctionCall(Normal=Normalize(VectorInput=@n6.Result), Flatness=ScalarParameter() {group=Normals, parameterName=NormalFlatten}) {materialFunction=&FlattenNormal}.Result
   Forward=@n14.Result
   Mobile=@n14.Result
  )
  MaterialFunctionCall(BaseNormal=@n15) {materialFunction=&BlendAngleCorrectedNormals}
  n16 = Multiply(A=@n15) {constA=1, desc=Normal}
  n17 = SceneDepth()
  n18 = Subtract(A=@n17.R, B=PixelDepth().R)
  n19 = ScalarParameter() {group=Caustics, parameterName=Min Refraction and Blur}
  n20 = Multiply(
   A=Divide(A=Max(A=@n18, B=@n19), B=@n17.R)
   B=ScalarParameter() {defaultValue=-1, desc=-0.4, group=Caustics, parameterName=Pixel Normal Offset}
  )
  n21 = Add(
   A=MaterialFunctionCall!(World Normal=Multiply(A=@n16, B=@n20), Custom UV Input) {materialFunction=/FortWater/Materials/Functions/Transform_Refraction_Normals}.View Space
   B=ScreenPosition().ViewportUV
  )
  n22 = MaterialFunctionCall(SceneDepth=SceneDepth(UVs=MaterialFunctionCall!(Screen Position=@n21) {desc=Dont pull in pixels from above the water, materialFunction=/FortWater/Materials/Functions/Distortion_MaskForeground}.Result).R) {materialFunction=&WorldPositionBehindTranslucency}
  n23 = ScalarParameter() {defaultValue=150, group=Caustics, parameterName=VatDepth}
  MaterialExpressionMultiply_28 = Multiply(A=CameraVectorWS()) {constB=-1}
  MaterialExpressionNormalize_1 = Normalize(VectorInput=Multiply(A=VectorParameter() {group=Caustics, parameterName=nynt, defaultValue=<rgba 0.5 0.5 0.25 0>}.RGB, B=@n15))
  MaterialExpressionScalarParameter_26 = ScalarParameter() {defaultValue=0.75, group=Caustics, parameterName=Relative Refractive Indices}
  n24 = Custom!(v=@MaterialExpressionMultiply_28, n=@MaterialExpressionNormalize_1, i=@MaterialExpressionScalarParameter_26) {
   code="return refract(v, n, i);"
   description=refract
   inputs [
    {inputName=v, input {expression=&self:MaterialExpressionMultiply_28}}
    {inputName=n, input {expression=&self:MaterialExpressionNormalize_1}}
    {inputName=i, input {expression=&self:MaterialExpressionScalarParameter_26}}
   ]
  }
  n25 = Multiply(
   A=Multiply(A=@n24, B=@n23) {desc=Refracted Vector * Depth}
   B=ScalarParameter() {defaultValue=-1, group=Caustics, parameterName=Caustic Refraction}
  )
  n26 = ScalarParameter() {defaultValue=0.1, parameterName=Time}
  n27 = VectorParameter() {group=Caustics, parameterName=Vat Velocity}
  Multiply(
   A=MaterialFunctionCall!(
    WorldPosition=Add(A=@n22.Result, B=@n25)
    Scale=ScalarParameter() {defaultValue=512, group=Caustics, parameterName=Caustics Scale}
    Water Depth=@n23
    Time=MaterialFunctionCall(Speed=@n26) {materialFunction=&DebugTimeSine}.Result
    Water Velocity=ComponentMask(None=@n27.RGB) {g=true, r=true}
   ) {materialFunction=/FortWater/Materials/Functions/WaterCaustics}.Result
   B=VectorParameter() {group=Caustics, parameterName=Caustics, defaultValue=<rgba 0.0022 0.0417 0.033 1>}.RGB
  )
  n28 = Multiply(A=ScalarParameter() {defaultValue=1, parameterName=FadeIn}) {constB=1.18}
  n29 = OneMinus(None=Multiply(A=Clamp(None=@n28), B=DecalLifetimeOpacity().Opacity))
  n30 = ScalarParameter() {defaultValue=-5, parameterName=ErodeStrength}
  n31 = TextureCoordinate() {uTiling=3, vTiling=3}
  n32 = TextureSample(UVs=MaterialFunctionCall(UVs=@n31) {materialFunction=&MF_DecalSphereCoords}.Result) {
   constMipValue=0
   mipValueMode=TMVM_MipLevel
   texture=/ChronoEnviroItems/LaunchablePlants/Gameplay/Effects/FX/Slurp/T_Chrono_Plant_Slurp_Drops
  }
  n33 = TextureSample!(UVs=@n31, Tex, Apply View MipBias) {texture=&T_Chrono_Plant_Slurp_Drops}
  n34 = TextureCoordinate() {uTiling=8, vTiling=8}
  n35 = Constant() {r=0.3}
  n36 = MaterialFunctionCall(
   UVs=MaterialFunctionCall(UVs=@n34) {materialFunction=&MF_DecalSphereCoords}.Result
   "Rotation Angle (0-1)"=@n35
  ) {materialFunction=&CustomRotator}
  n37 = TextureSample!(
   UVs=MaterialFunctionCall(UVs=@n34, "Rotation Angle (0-1)"=@n35) {materialFunction=&CustomRotator}.Rotated Values
   Tex
   Level
   Apply View MipBias
  ) {
   constMipValue=0
   mipValueMode=TMVM_MipLevel
   texture=&T_Chrono_Plant_Slurp_Drops
  }
  n38 = Max(
   A=ShadingPathSwitch(Default=@n32.R, Forward=@n33.R, Mobile=@n33.R)
   B=ShadingPathSwitch(
    Default=TextureSample!(UVs=@n36.Rotated Values, Tex, Apply View MipBias) {texture=&T_Chrono_Plant_Slurp_Drops}.R
    Forward=@n37.R
    Mobile=@n37.R
   )
  )
  n39 = TextureCoordinate() {uTiling=2, vTiling=2}
  n40 = TextureSample(UVs=MaterialFunctionCall(UVs=@n39) {materialFunction=&MF_DecalSphereCoords}.Result) {
   samplerSource=SSM_Wrap_WorldGroupSettings
   samplerType=SAMPLERTYPE_LinearColor
   texture=/Game/Effects/Fort_Effects/Textures/Water/T_waterSplash
  }
  n41 = TextureSample(UVs=@n39) {
   samplerSource=SSM_Wrap_WorldGroupSettings
   samplerType=SAMPLERTYPE_LinearColor
   texture=&T_waterSplash
  }
  n42 = TextureCoordinate() {uTiling=4, vTiling=4}
  n43 = Constant() {r=0.45}
  n44 = MaterialFunctionCall(
   UVs=MaterialFunctionCall(UVs=@n42) {materialFunction=&MF_DecalSphereCoords}.Result
   "Rotation Angle (0-1)"=@n43
  ) {materialFunction=&CustomRotator}
  n45 = Constant2Vector() {r=0.5}
  n46 = TextureSample(UVs=Add(A=Multiply(A=@n44.Rotated Values) {constB=0.5}, B=@n45)) {
   samplerSource=SSM_Wrap_WorldGroupSettings
   samplerType=SAMPLERTYPE_LinearColor
   texture=&T_waterSplash
  }
  n47 = MaterialFunctionCall(UVs=@n42, "Rotation Angle (0-1)"=@n43) {materialFunction=&CustomRotator}
  n48 = TextureSample(UVs=Add(A=Multiply(A=@n47.Rotated Values) {constB=0.5}, B=@n45)) {
   samplerSource=SSM_Wrap_WorldGroupSettings
   samplerType=SAMPLERTYPE_LinearColor
   texture=&T_waterSplash
  }
  n49 = Max(A=ShadingPathSwitch(Default=@n40.G, Forward=@n41.G, Mobile=@n41.G), B=ShadingPathSwitch(Default=@n46.G, Forward=@n48.G, Mobile=@n48.G))
  n50 = TextureSample(UVs=Add(A=TextureCoordinate() {uTiling=0.75, vTiling=0.75}) {constB=0.125}) {
   samplerSource=SSM_Clamp_WorldGroupSettings
   samplerType=SAMPLERTYPE_LinearGrayscale
   texture=/Game/Environments/Artemis/ExplosionCrater/Textures/ZoomBoom_M
  }
  n51 = Multiply(
   A=Power(Base=@n50.R, Exp=ScalarParameter() {defaultValue=2, parameterName=Exp}) {constExponent=10}
   B=ScalarParameter() {defaultValue=1, parameterName=Multi}
  )
  n52 = Subtract(A=Clamp(None=Max(A=@n38, B=@n49)), B=OneMinus(None=Clamp(None=@n51)))
  n53 = SphereMask(A=LightVector(), B=Constant() {r=0.5}, Radius=Constant() {r=0.5}, Hardness=ScalarParameter() {defaultValue=1, parameterName=Sphere_Hardness}) {hardnessPercent=75}
  Multiply(
   A=MaterialFunctionCall!(Dynamic Param=Multiply(A=@n29, B=@n30), Alpha=Multiply(A=@n52, B=@n53)) {materialFunction=/Game/Effects/Fort_Effects/Materials/_Functions/Erode_Alpha}.Result
   B=ScalarParameter() {defaultValue=0.25, parameterName=MasterOpacity}
  )
  n54 = TextureCoordinate() {uTiling=2, vTiling=2}
  n55 = TextureSample!(
   UVs=MaterialFunctionCall(UVs=@n54) {materialFunction=&MF_DecalSphereCoords}.Result
   Tex
   Level
   Apply View MipBias
  ) {
   constMipValue=0
   mipValueMode=TMVM_MipLevel
   samplerSource=SSM_Wrap_WorldGroupSettings
   samplerType=SAMPLERTYPE_Normal
   texture=/Game/Effects/Fort_Effects/Textures/Water/T_waterSplash_N
  }
  n56 = TextureSample!(UVs=@n54, Tex, Apply View MipBias) {samplerSource=SSM_Wrap_WorldGroupSettings, samplerType=SAMPLERTYPE_Normal, texture=&T_waterSplash_N}
  n57 = ShadingPathSwitch(Default=@n55.RGB, Forward=@n56.RGB, Mobile=@n56.RGB)
  n58 = TextureCoordinate() {uTiling=4, vTiling=4}
  n59 = Constant() {r=0.45}
  n60 = MaterialFunctionCall(
   UVs=MaterialFunctionCall(UVs=@n58) {materialFunction=&MF_DecalSphereCoords}.Result
   "Rotation Angle (0-1)"=@n59
  ) {materialFunction=&CustomRotator}
  n61 = Constant2Vector() {r=0.5}
  n62 = TextureSample(UVs=Add(A=Multiply(A=@n60.Rotated Values) {constB=0.5}, B=@n61)) {samplerSource=SSM_Wrap_WorldGroupSettings, samplerType=SAMPLERTYPE_Normal, texture=&T_waterSplash_N}
  n63 = MaterialFunctionCall(UVs=@n58, "Rotation Angle (0-1)"=@n59) {materialFunction=&CustomRotator}
  n64 = TextureSample!(UVs=Add(A=Multiply(A=@n63.Rotated Values) {constB=0.5}, B=@n61), Tex, Level, Apply View MipBias) {
   constMipValue=0
   mipValueMode=TMVM_MipLevel
   samplerSource=SSM_Wrap_WorldGroupSettings
   samplerType=SAMPLERTYPE_Normal
   texture=&T_waterSplash_N
  }
  n65 = ShadingPathSwitch(Default=@n62.RGB, Forward=@n64.RGB, Mobile=@n64.RGB)
  n66 = TextureCoordinate() {uTiling=3, vTiling=3}
  n67 = TextureSample(UVs=MaterialFunctionCall(UVs=@n66) {materialFunction=&MF_DecalSphereCoords}.Result) {
   constMipValue=0
   mipValueMode=TMVM_MipLevel
   samplerType=SAMPLERTYPE_Normal
   texture=/ChronoEnviroItems/LaunchablePlants/Gameplay/Effects/FX/Slurp/T_Chrono_Plant_Slurp_Drops_N
  }
  n68 = TextureSample!(UVs=@n66, Tex, Apply View MipBias) {samplerType=SAMPLERTYPE_Normal, texture=&T_Chrono_Plant_Slurp_Drops_N}
  n69 = ShadingPathSwitch(Default=@n67.R, Forward=@n68.R, Mobile=@n68.R)
  n70 = TextureCoordinate() {uTiling=8, vTiling=8}
  n71 = MaterialFunctionCall(UVs=@n70) {materialFunction=&MF_DecalSphereCoords}
  n72 = Constant() {r=0.3}
  n73 = TextureSample!(
   UVs=MaterialFunctionCall(UVs=@n70, "Rotation Angle (0-1)"=@n72) {materialFunction=&CustomRotator}.Rotated Values
   Tex
   Level
   Apply View MipBias
  ) {
   constMipValue=0
   mipValueMode=TMVM_MipLevel
   samplerType=SAMPLERTYPE_Normal
   texture=&T_Chrono_Plant_Slurp_Drops_N
  }
  n74 = ShadingPathSwitch(
   Default=TextureSample!(
    UVs=MaterialFunctionCall(UVs=@n71.Result, "Rotation Angle (0-1)"=@n72) {materialFunction=&CustomRotator}.Rotated Values
    Tex
    Apply View MipBias
   ) {samplerType=SAMPLERTYPE_Normal, texture=&T_Chrono_Plant_Slurp_Drops_N}.R
   Forward=@n73.R
   Mobile=@n73.R
  )
  Normalize(VectorInput=Add(A=Add(A=@n57, B=@n65), B=Add(A=@n69, B=@n74)))
  n75 = ComponentMask(None=@n15) {b=true}
  n76 = Power(Base=@n75, Exp=Constant() {r=10})
  Reroute(None=LinearInterpolate(
    A=VectorParameter() {group=Color, parameterName=ColorA, defaultValue=<rgba 0 0.4488 1 1>}.RGB
    B=VectorParameter() {group=Color, parameterName=ColorB, defaultValue=<rgba 0.0251 0.4688 0.371 1>}.RGB
    Alpha=Saturate(None=@n76)
   ))
  Reroute(None=LinearInterpolate(Alpha=@n75) {constA=0.1, constB=0.07})
  Reroute(None=LinearInterpolate(Alpha=@n75) {constB=0.5})
 ]
}
