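# Niagara system for the GravBerry cactus slurp explosion, bursting 15 fruit bulbs into glows, splashes, drops and a fading two-layer ground decal.
# Particles spawn at static mesh sockets, indexed against a Fruit Explode Times array to stagger each pop.
# GPU splashes and drops share one emitter via spawn groups, with quality-scaled counts and distance culling.
# User.WasHitByVeh removes decal and splash spawn delays; several older or debug emitters are disabled.
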NiagaraSystem /CactusFruits/VFX/NS_GravBerryCactus_SlurpExplosion
dependencies=null
emitters {
 Debugging {
  bEnabled=false
  emitterUpdateScript [
   &EmitterState()
   &SpawnBurst_Instantaneous(Use Spawn Probability=false, UseLoopCountLimit=false)
  ]
  particleSpawnScript [&InitializeParticle(Color Mode=::NewEnumerator0, Lifetime=5, Position Offset=<0 0 50>, UsePositionOffset=true, Write Material Random=true)]
  particleUpdateScript [
   &ParticleState()
   &DynamicMaterialParameters(
    Index 0 Param 1=$Engine.Owner.LODDistance
    Module.Index0Param0Write=true
    Module.Index0Param1Write=true
    Module.Index0Param2Write=true
    Module.Index0Param3Write=true
   )
  ]
  rendererClasses [&NiagaraSpriteRendererProperties]
  renderers [{rendererClass=&NiagaraSpriteRendererProperties}]
 }
 Decal {
  emitterUpdateScript [
   &EmitterState(
    Inactive Response=::NewEnumerator1
    Life Cycle Mode=::NewEnumerator1
    Loop Behavior=::NewEnumerator1
    Loop Delay=&MakeCustomFloatFromBool(Bool=$User.WasHitByVeh, True Float=0, False Float=0.2)
    Loop Duration=7
    UseLoopDelay=true
   )
   &SpawnBurst_Instantaneous(
    Spawn Time=&MakeCustomFloatFromBool(Bool=$User.WasHitByVeh, True Float=0, False Float=0.2)
    Use Spawn Probability=false
    UseLoopCountLimit=false
   )
  ]
  particleSpawnScript [
   &InitializeParticle(
    Lifetime=10
    Mesh Scale=&Multiply_VectorByFloat(Vector=<50 800 800>, Float=1.1)
    Mesh Scale Mode=::NewEnumerator4
    UsePositionOffset=false
    Write Material Random=true
   )
   &InitialMeshOrientation(
    Mesh Axis To Orient=<0 0 1>
    Mesh Orientation Mode=::NewEnumerator2
    Rotation=&RandomRangeVector(Minimum=<0 0 0>, Maximum=<1 0 0>)
    Use Rotation Vector=true
   )
  ]
  particleUpdateScript [
   &ParticleState(Kill Particles When Lifetime Has Elapsed=false)
   &ScaleColor(ScaleA=true, ScaleRGB=true)
   /SetVariables(Particles.DecalSize:Vector3f=$Particles.Scale) {isSetParameters=true}
   /SetVariables(Particles.DecalSize2:Vector3f=&Multiply_Vector(A=$Particles.Scale, B=<0.1 0.5 0.5>)) {isSetParameters=true}
   /SetVariables(Particles.DecalFade:NiagaraFloat=&FloatFromCurve(FloatCurve=DataInterface() {propertyValues {Curve {keys [
         {arriveTangent=9.0725, interpMode=RCIM_Cubic, time=0, value=0}
         <at 0.1 0.9847 RCIM_Cubic>
         {arriveTangent=-2.8714, interpMode=RCIM_Cubic, leaveTangent=-2.8714, tangentMode=RCTM_User, time=1, value=0}
        ]}}})) {isSetParameters=true}
   /SetVariables(Particles.DecalFade2:NiagaraFloat=&FloatFromCurve(FloatCurve=DataInterface() {propertyValues {Curve {keys [
         {
          arriveTangent=18.6734
          arriveTangentWeight=0.2822
          interpMode=RCIM_Cubic
          leaveTangent=18.6734
          leaveTangentWeight=0.2822
          tangentMode=RCTM_User
          tangentWeightMode=RCTWM_WeightedBoth
          time=0.0502
          value=0
         }
         {
          arriveTangentWeight=0.069
          interpMode=RCIM_Cubic
          leaveTangentWeight=0.1833
          tangentMode=RCTM_Break
          tangentWeightMode=RCTWM_WeightedBoth
          time=0.1502
          value=0.9847
         }
         {arriveTangent=-2.8714, interpMode=RCIM_Cubic, leaveTangent=-2.8714, tangentMode=RCTM_User, time=0.65, value=0}
        ]}}})) {isSetParameters=true}
   /SetVariables(Particles.DecalOrientation:Quat4f=$Particles.MeshOrientation) {isSetParameters=true}
  ]
  rendererClasses [&NiagaraDecalRendererProperties]
  renderers [
   {rendererClass=&NiagaraDecalRendererProperties, rendererIndex=0}
   {rendererClass=&NiagaraDecalRendererProperties, rendererIndex=1}
  ]
 }
 Empty {
  bEnabled=false
  emitterSpawnScript [/SetVariables(
    Emitter.Number Of Bulbs=15
    Emitter.Particles Per Bulb:NiagaraInt32=&MakeCustomIntFromBool(
     Bool=&IsPlatformSetActive(New Platform Set=DataInterface() {propertyValues {Platforms {qualityLevelMask=-31}}})
     True INT=1
     False INT=&MakeCustomIntFromBool(
      Bool=&IsPlatformSetActive(New Platform Set=DataInterface() {propertyValues {Platforms {qualityLevelMask=-30}}})
      True INT=4
      False INT=&MakeCustomIntFromBool(
       Bool=&IsPlatformSetActive(New Platform Set=DataInterface() {propertyValues {Platforms {qualityLevelMask=-28}}})
       True INT=5
       False INT=5
      )
     )
    )
   ) {isSetParameters=true}]
  emitterUpdateScript [
   &EmitterState()
   &SpawnBurst_Instantaneous(
    Spawn Count=$Emitter.Particles Per Bulb
    Spawn Time=&MakeCustomFloatFromBool(Bool=$User.WasHitByVeh, True Float=0, False Float=0.05)
    Use Spawn Probability=false
    UseLoopCountLimit=false
   )
  ]
  particleSpawnScript [
   &InitializeParticle(
    Color Maximum=<rgba 4 4 4 1>
    Color Minimum=<rgba 3 3 3 1>
    Color Mode=::NewEnumerator2
    Lifetime Max=0.75
    Lifetime Min=0.5
    Lifetime Mode=::NewEnumerator1
    Sprite Rotation Mode=::NewEnumerator2
    Sprite Size Mode=::NewEnumerator1
    Uniform Sprite Size Max=7
    Uniform Sprite Size Min=10
    UsePositionOffset=false
    Write Material Random=true
   )
   &ShapeLocation(
    MaskSpawnGroup=false
    Non Uniform Scale=<1 1 2>
    Offset=<0 0 50>
    Offset Mode=::NewEnumerator0
    Override Seed=false
    Sphere Radius=5
    Sphere Surface Distribution=1
    UseOwnerScale=true
   )
   &AddVelocityFromPoint(
    Module.Use Falloff Distance=false
    Origin Offset=<0 0 52>
    Velocity Strength=&UniformRangedFloat(Minimum=500, Maximum=200)
   )
   &AddVelocity(Velocity=<0 0 400>)
   &DynamicMaterialParameters(
    Index 0 Param 1=&RandomRangeFloat(Minimum=1, Maximum=2)
    Index 0 Param 2=0.5
    Index 0 Param 3=&RandomRangeFloat(Minimum=15, Maximum=50)
    Index 0 Param 4=&RandomRangeFloat(Minimum=2, Maximum=10)
    Index 1 Param 1=&RandomRangeFloat(Minimum=0, Maximum=10)
    Module.Param0WriteEnabled=true
    Module.Param0WriteEnabled1=true
    Module.Param1WriteEnabled=true
    Module.Param1WriteEnabled1=false
    Module.Param2WriteEnabled=true
    Module.Param2WriteEnabled1=false
    Module.Param3WriteEnabled=true
    Module.Param3WriteEnabled1=false
    Write Parameter Index 1=true
   )
   /Niagara/Modules/Spawn/Location/StaticMeshLocation(
    Interpolate Mesh Sampling=false
    MaskSpawnGroup=false
    Mesh Sampling Type:ENiagara_StaticSamplingMode=::NewEnumerator1
    SamplePosOffset=true
    Sampled Position Offset:Vector3f=<0 0 -7>
    Sampled Position Offset Coordinate Space:ENiagaraSkelMeshTransforms=::NewEnumerator1
    Socket Index:NiagaraInt32=&SpawnGroup()
    Socket Sampling Mode:ENiagara_StaticSocketMode=::NewEnumerator2
    Static Mesh:NiagaraDataInterfaceStaticMesh=$System.Static Mesh
   )
  ]
  rendererClasses []
  renderers []
 }
 FruitGlowsV2 {
  emitterSpawnScript [
   /SetVariables(Emitter.Particles Per Bulb=1) {isSetParameters=true, module=SetVariables_#1}
   /SetVariables(Emitter.SpawnTimeMultiplier:NiagaraFloat=&MakeCustomFloatFromBool(Bool=$User.WasHitByVeh, True Float=0.1, False Float=0.2)) {enabled=false, isSetParameters=true}
  ]
  emitterUpdateScript [
   &EmitterState(
    Enable Distance Culling=true
    Life Cycle Mode=::NewEnumerator1
    Loop Behavior=::NewEnumerator1
    Loop Duration=1
    MaxDistance=4000
    MinDistance=3000
    Module.bUseMaxDistance=true
    Module.bUseMinDistance=false
    Scalability Mode=::NewEnumerator1
    UseLoopDelay=true
   )
   &SpawnBurst_Instantaneous(Spawn Count=15, Use Spawn Probability=false, UseLoopCountLimit=false)
  ]
  particleSpawnScript [
   &InitializeParticle(
    Color=<rgba 0.05 0.05 0.05 1>
    Lifetime=0.35
    Mass Max=2
    Mass Mode=::NewEnumerator3
    Sprite Rotation Mode=::NewEnumerator2
    Sprite Size Mode=::NewEnumerator1
    Sprite UV Mode=::NewEnumerator3
    Uniform Sprite Size Max=50
    Uniform Sprite Size Min=50
    UsePositionOffset=false
    Write Material Random=true
   )
   &StaticMeshLocation(
    Interpolate Mesh Sampling=false
    MaskSpawnGroup=false
    Mesh Sampling Type:ENiagara_StaticSamplingMode=::NewEnumerator1
    SamplePosOffset=true
    Sampled Position Offset:Vector3f=<7 0 0>
    Socket Index:NiagaraInt32=&ReturnParticleIDAttributes()
    Socket Sampling Mode:ENiagara_StaticSocketMode=::NewEnumerator2
    Static Mesh:NiagaraDataInterfaceStaticMesh=$System.Static Mesh
   )
   &CameraOffset(Camera Offset Amount=10)
   /SetVariables(Particles.Position=&AddVectorToPosition(Position=$Particles.Position, Vector=<0 0 -5>)) {isSetParameters=true}
  ]
  particleUpdateScript [
   &ParticleState()
   &SpriteRotationRate(Rotation Rate=&RandomRangeFloat(Minimum=-180, Maximum=180))
   &ScaleSpriteSize(
    Uniform Curve Index=&Subtract_Float(
     A=$Particles.Age
     B=&SelectFloatFromArray(Direct Array Index=&ReturnParticleIDAttributes(), Float Selection Array=$System.Fruit Explode Times)
    )
    Uniform Curve Scale=1.5
    Uniform Curve Sprite Scale=DataInterface() {propertyValues {Curve {keys [
        {arriveTangent=13.902, interpMode=RCIM_Cubic, leaveTangent=13.902, tangentMode=RCTM_User, time=0, value=0.35}
        {interpMode=RCIM_Cubic, leaveTangent=-0.5754, time=0.175, value=2}
       ]}}}
   )
   &ScaleColor(
    Scale Alpha=&FloatFromCurve(
     FloatCurve=DataInterface() {propertyValues {Curve {keys [
         <at 0 -1>
         <at 0.0002 0.5 RCIM_Cubic>
         {arriveTangent=-19.0556, interpMode=RCIM_Cubic, leaveTangent=-19.0556, tangentMode=RCTM_User, time=0.175, value=-1}
        ]}}}
     CurveIndex=&Subtract_Float(
      A=$Particles.Age
      B=&SelectFloatFromArray(Direct Array Index=&ReturnParticleIDAttributes(), Float Selection Array=$System.Fruit Explode Times)
     )
    )
    Scale RGB=&MakeVectorFromLinearColorRGB(Color=&MultiplyLinearColor(
      Linear Color=&ColorFromCurve(
       ColorCurve=DataInterface() {propertyValues {
         AlphaCurve {keys []}
         BlueCurve {keys [<at 0.024 40>, {arriveTangent=-256.3662, interpMode=RCIM_Cubic, leaveTangent=-256.3662, time=0.0321, value=12}, <at 0.1488 8>, <at 0.1615 8>]}
         GreenCurve {keys [
           <at 0.024 28.1816>
           {arriveTangent=-161.6839, interpMode=RCIM_Cubic, leaveTangent=-161.6839, time=0.0321, value=12}
           <at 0.1488 8>
           <at 0.1615 8>
          ]}
         RedCurve {keys [<at 0.024 7.2711>, {arriveTangent=5.8395, interpMode=RCIM_Cubic, leaveTangent=5.8395, time=0.0321, value=12}, <at 0.1488 8>, <at 0.1615 8>]}
        }}
       CurveIndex=&Subtract_Float(
        A=$Particles.Age
        B=&SelectFloatFromArray(Direct Array Index=&ReturnParticleIDAttributes(), Float Selection Array=$System.Fruit Explode Times)
       )
      )
      Scale Factor=&MakeVector4_FromLinearColor(LinearColor=&LerpLinearColors(StartColor=<rgba 0.0189 5.7031 6.9792 1>, EndColor=<rgba 0.0196 7.2396 6.3973 1>, LerpFactor=$Particles.MaterialRandom))
     ))
    ScaleA=true
    ScaleRGB=true
   )
  ]
  rendererClasses [&NiagaraSpriteRendererProperties]
  renderers [{rendererClass=&NiagaraSpriteRendererProperties}]
 }
 FruitSplashesAndDrops {
  bEnabled=false
  simTarget=GPUComputeSim
  emitterSpawnScript [
   /SetVariables(Emitter.Splashes Per Bulb:NiagaraInt32=&MakeCustomIntFromBool(
     Bool=&IsPlatformSetActive(New Platform Set=DataInterface() {propertyValues {Platforms {qualityLevelMask=-31}}})
     True INT=1
     False INT=&MakeCustomIntFromBool(
      Bool=&IsPlatformSetActive(New Platform Set=DataInterface() {propertyValues {Platforms {qualityLevelMask=-30}}})
      True INT=1
      False INT=&MakeCustomIntFromBool(
       Bool=&IsPlatformSetActive(New Platform Set=DataInterface() {propertyValues {Platforms {qualityLevelMask=-28}}})
       True INT=&RandomRangeInt(Minimum=1, Maximum=2)
       False INT=&RandomRangeInt(Minimum=2, Maximum=4)
      )
     )
    )) {isSetParameters=true, module=SetVariables_#1}
   /SetVariables(Emitter.Drops Per Bulb:NiagaraInt32=&MakeCustomIntFromBool(
     Bool=&IsPlatformSetActive(New Platform Set=DataInterface() {propertyValues {Platforms {qualityLevelMask=-31}}})
     True INT=1
     False INT=&MakeCustomIntFromBool(
      Bool=&IsPlatformSetActive(New Platform Set=DataInterface() {propertyValues {Platforms {qualityLevelMask=-30}}})
      True INT=&RandomRangeInt(Minimum=1, Maximum=3)
      False INT=&MakeCustomIntFromBool(
       Bool=&IsPlatformSetActive(New Platform Set=DataInterface() {propertyValues {Platforms {qualityLevelMask=-28}}})
       True INT=&RandomRangeInt(Minimum=2, Maximum=4)
       False INT=&RandomRangeInt(Minimum=3, Maximum=5)
      )
     )
    )) {isSetParameters=true, module=SetVariables_#2}
   /SetVariables(Emitter.Goop Spawn Scale=0.5) {enabled=false, isSetParameters=true, module=SetVariables_#3}
   /SetVariables(Emitter.Drop Spawn Scale=0.5) {enabled=false, isSetParameters=true, module=SetVariables_#4}
  ]
  emitterUpdateScript [
   &EmitterState(
    Enable Distance Culling=true
    Life Cycle Mode=::NewEnumerator1
    Loop Behavior=::NewEnumerator1
    Loop Duration=1
    MaxDistance=2800
    Module.bUseMaxDistance=true
    Module.bUseMinDistance=false
    Scalability Mode=::NewEnumerator1
    UseLoopDelay=false
   )
   &SpawnBurst_Instantaneous(
    Spawn Count=&Multiply_Int(A=$Emitter.Splashes Per Bulb, B=15)
    Spawn Time=&MakeCustomFloatFromBool(Bool=$User.WasHitByVeh, True Float=0, False Float=0.05)
    Use Spawn Probability=false
    UseLoopCountLimit=false
   )
   &SpawnBurst_Instantaneous(
    Spawn Count=&Multiply_Int(A=$Emitter.Drops Per Bulb, B=15)
    Spawn Group=15
    Spawn Probability=&MakeCustomFloatFromBool(
     Bool=&IsPlatformSetActive(New Platform Set=DataInterface() {propertyValues {Platforms {qualityLevelMask=-29}}})
     True Float=0
     False Float=1
    )
    Spawn Time=&MakeCustomFloatFromBool(Bool=$User.WasHitByVeh, True Float=0, False Float=0.05)
    Use Spawn Probability=true
    UseLoopCountLimit=false
   ) {module=SpawnBurst_Instantaneous015}
  ]
  particleSpawnScript [
   /SetVariables(Particles.bIsSplashes:NiagaraBool=/Niagara/DynamicInputs/Helpers/MaskBoolBySpawnGroup(Masked Bool:NiagaraBool=true, Passthrough Bool:NiagaraBool=false, Spawn Group Mask:NiagaraInt32=0)) {isSetParameters=true, module=SetVariables_#5}
   &InitializeParticle(
    Color=&MakeCustomLinearColorFromBool(
     Bool=$Particles.bIsSplashes
     True Color=&RandomRangeLinearColor(Minimum=<rgba 4 4 4 1>, Maximum=<rgba 0.7835 0.7835 0.7835 0.6>)
     False Color=&RandomRangeLinearColor(Minimum=<rgba 3 3 3 1>, Maximum=<rgba 4 4 4 1>)
    )
    Lifetime=&Add_Float(
     A=&MakeCustomFloatFromBool(
      Bool=$Particles.bIsSplashes
      True Float=&RandomRangeFloat(Minimum=0.85, Maximum=1.5)
      False Float=&RandomRangeFloat(Minimum=0.5, Maximum=0.75)
     )
     B=0
    )
    Mass Max=2
    Mass Mode=::NewEnumerator3
    Sprite Rotation Mode=::NewEnumerator2
    Sprite Size Mode=::NewEnumerator3
    Sprite UV Mode=::NewEnumerator3
    Uniform Sprite Size=&MakeCustomFloatFromBool(
     Bool=$Particles.bIsSplashes
     True Float=&RandomRangeFloat(Minimum=50, Maximum=75)
     False Float=&RandomRangeFloat(Minimum=7, Maximum=10)
    )
    UsePositionOffset=false
    Write Material Random=true
   )
   &ShapeLocation(
    MaskSpawnGroup=false
    Non Uniform Scale=&MakeCustomVectorFromBool(Bool=$Particles.bIsSplashes, True Vector=<0 0 0>, False Vector=<1 1 2>)
    Offset=&MakeCustomVectorFromBool(Bool=$Particles.bIsSplashes, True Vector=<0 0 0>, False Vector=<0 0 50>)
    Offset Mode=::NewEnumerator0
    Override Seed=false
    Sphere Radius=&MakeCustomFloatFromBool(Bool=$Particles.bIsSplashes, True Float=0, False Float=5)
    Sphere Surface Distribution=1
    UseOwnerScale=true
   )
   &AddVelocityFromPoint(
    Module.Use Falloff Distance=false
    Origin Offset=<0 0 52>
    Velocity Strength=&MakeCustomFloatFromBool(Bool=$Particles.bIsSplashes, True Float=0, False Float=&UniformRangedFloat(Minimum=500, Maximum=200))
   )
   &AddVelocity(Velocity=&MakeCustomVectorFromBool(Bool=$Particles.bIsSplashes, True Vector=<0 0 0>, False Vector=<0 0 400>))
   &StaticMeshLocation(
    Interpolate Mesh Sampling=false
    MaskSpawnGroup=false
    Mesh Sampling Type:ENiagara_StaticSamplingMode=::NewEnumerator1
    SamplePosOffset=true
    Sampled Position Offset:Vector3f=<0 0 -7>
    Sampled Position Offset Coordinate Space:ENiagaraSkelMeshTransforms=::NewEnumerator1
    Socket Index:NiagaraInt32=&ModuloInt(A=&ReturnParticleIDAttributes(), B=15)
    Socket Sampling Mode:ENiagara_StaticSocketMode=::NewEnumerator2
    Static Mesh:NiagaraDataInterfaceStaticMesh=$System.Static Mesh
   )
   &CameraOffset(Camera Offset Amount=&MakeCustomFloatFromBool(Bool=$Particles.bIsSplashes, True Float=15, False Float=0))
   &AddVelocity(
    Cone Angle=60
    Cone Axis=&ReturnMeshOrientationAxis(Mesh Orientation Axis=::XAxis, Mesh Orientation Quat=$Particles.StaticMeshLocation.SampledRotation)
    Module.UseConeFalloff=false
    Module.UseVelDistribution=false
    Override Seed=false
    Velocity Mode=::NewEnumerator2
    Velocity Speed=&MakeCustomFloatFromBool(Bool=$Particles.bIsSplashes, True Float=600, False Float=0)
   ) {module=AddVelocity003}
   &DynamicMaterialParameters(
    Index 0 Param 1=&RandomRangeFloat(Minimum=1, Maximum=2)
    Index 0 Param 2=0.5
    Index 0 Param 3=&RandomRangeFloat(Minimum=15, Maximum=50)
    Index 0 Param 4=&RandomRangeFloat(Minimum=2, Maximum=10)
    Index 1 Param 1=&RandomRangeFloat(Minimum=0, Maximum=10)
    Module.Param0WriteEnabled=true
    Module.Param0WriteEnabled1=true
    Module.Param1WriteEnabled=true
    Module.Param1WriteEnabled1=false
    Module.Param2WriteEnabled=true
    Module.Param2WriteEnabled1=false
    Module.Param3WriteEnabled=true
    Module.Param3WriteEnabled1=false
    Write Parameter Index 1=true
   )
   /SetVariables(Particles.VisibilityTag=&MakeCustomIntFromBool(Bool=$Particles.bIsSplashes, True INT=0, False INT=1)) {isSetParameters=true, module=SetVariables_#6}
   /SetVariables(Particles.accumulatedVelocity:Vector3f=$Particles.Velocity) {isSetParameters=true, module=SetVariables_#7}
   /SetVariables() {isSetParameters=true, module=SetVariables_#8}
  ]
  particleUpdateScript [
   &ParticleState()
   &SpriteRotationRate(Rotation Rate=&MakeCustomFloatFromBool(
     Bool=$Particles.bIsSplashes
     True Float=&Multiply_Float(
      A=&Lerp_Float(
       A=180
       B=45
       Alpha=&RemapRange(Input Value To Remap=$Particles.Mass, Input Min=0.75, Input Max=2, Output Min=0, Output Max=1)
      )
      B=&MakeFloatFromInt(Int=&RandomRangeInt(Minimum=-1, Maximum=1))
     )
     False Float=0
    ))
   &DoOnce(Trigger Condition:NiagaraBool=&SetBoolByFloatComparison(
     A=$Emitter.Age
     B=&SelectFloatFromArray(
      Direct Array Index=&ModuloInt(A=&ReturnParticleIDAttributes(), B=15)
      Direct Array Mode=::NewEnumerator1
      Float Selection Array=$System.Fruit Explode Times
     )
    ))
   &Drag(
    Drag=&MakeCustomFloatFromBool(
     Bool=$Particles.bIsSplashes
     True Float=&FloatFromCurve(
      FloatCurve=DataInterface() {propertyValues {Curve {keys [<at 0 2 RCIM_Cubic>,<at 0.5 0>]}}}
      Scale Curve=&RandomRangeFloat(Minimum=2, Maximum=2.5)
     )
     False Float=&FloatFromCurve(
      FloatCurve=DataInterface() {propertyValues {Curve {keys [<at 0 1 RCIM_Cubic>, {arriveTangent=-1.4297, interpMode=RCIM_Cubic, leaveTangent=-1.4297, tangentMode=RCTM_User, time=1, value=0}]}}}
      Scale Curve=&UniformRangedFloat(Minimum=10, Maximum=8)
     )
    )
    Use Linear Drag=true
    Use Rotational Drag=true
   )
   &GravityForce(Gravity=&Multiply_VectorByFloat(
     Vector=&MakeCustomVectorFromBool(Bool=$Particles.bIsSplashes, True Vector=<0 0 -980>, False Vector=<0 0 -1470>)
     Float=&MakeCustomFloatFromBool(Bool=$Particles.DoOnce.TriggeredOnce, True Float=1, False Float=0)
    ))
   &SolveForcesAndVelocity(Clamp Velocity=false, Limit Acceleration=false)
   &ScaleSpriteSizeBySpeed(
    Max Scale Factor=&MakeCustomVector2DFromBool(Bool=$Particles.bIsSplashes, True Vector2D=<1 1>, False Vector2D=<1.2 4>)
    Min Scale Factor=&MakeCustomVector2DFromBool(Bool=$Particles.bIsSplashes, True Vector2D=<1 1>, False Vector2D=<0.25 0.2>)
   )
   &ScaleSpriteSize(
    Scale Sprite Size Mode=::NewEnumerator0
    Uniform Scale Factor=&MakeCustomFloatFromBool(
     Bool=$Particles.bIsSplashes
     True Float=&FloatFromCurve(
      FloatCurve=DataInterface() {propertyValues {Curve {keys [
          <at 0 0 RCIM_Cubic>
          {arriveTangent=1.8917, interpMode=RCIM_Cubic, leaveTangent=1.8917, tangentMode=RCTM_User, time=0.001, value=0.35}
          {interpMode=RCIM_Cubic, leaveTangent=-0.5754, time=1, value=2}
         ]}}}
      CurveIndex=&Subtract_Float(
       A=&Multiply_Float(A=$Particles.Age, B=&Divide_Float(A=1, B=$Particles.Lifetime))
       B=&Multiply_Float(
        A=&SelectFloatFromArray(
         Direct Array Index=&ModuloInt(A=&ReturnParticleIDAttributes(), B=15)
         Float Selection Array=$System.Fruit Explode Times
        )
        B=1
       )
      )
      Scale Curve=$Particles.Lifetime
     )
     False Float=1
    )
   )
   &ScaleColor(
    Scale Alpha=&MakeCustomFloatFromBool(
     Bool=$Particles.bIsSplashes
     True Float=&FloatFromCurve(
      FloatCurve=DataInterface() {propertyValues {Curve {keys [
          <at 0 -1>
          {arriveTangentWeight=0.0001, interpMode=RCIM_Cubic, leaveTangentWeight=0.3333, tangentMode=RCTM_User, time=0.0002, value=0.5}
          {arriveTangent=-2.4995, arriveTangentWeight=0.8972, interpMode=RCIM_Cubic, leaveTangent=-2.4995, tangentMode=RCTM_User, time=1, value=-1}
         ]}}}
      CurveIndex=&Subtract_Float(
       A=&Multiply_Float(A=$Particles.Age, B=&Divide_Float(A=1, B=$Particles.Lifetime))
       B=&Multiply_Float(
        A=&SelectFloatFromArray(
         Direct Array Index=&ModuloInt(A=&ReturnParticleIDAttributes(), B=15)
         Float Selection Array=$System.Fruit Explode Times
        )
        B=$Particles.Lifetime
       )
      )
     )
     False Float=&FloatFromCurve(FloatCurve=DataInterface() {propertyValues {Curve {keys [<at 0 1 RCIM_Cubic>, {arriveTangent=-3.0079, interpMode=RCIM_Cubic, leaveTangent=-3.0079, tangentMode=RCTM_User, time=1, value=0}]}}})
    )
    Scale RGB=&MakeCustomVectorFromBool(
     Bool=$Particles.bIsSplashes
     True Vector=&MakeVectorFromLinearColorRGB(Color=&MultiplyLinearColor(
       Linear Color=&ColorFromCurve(
        ColorCurve=DataInterface() {propertyValues {
          AlphaCurve {keys []}
          BlueCurve {keys [
            <at 0.1481 40>
            {
             arriveTangent=-41.3655
             arriveTangentWeight=0.0167
             interpMode=RCIM_Cubic
             leaveTangent=-41.3655
             tangentMode=RCTM_User
             time=0.198
             value=12
            }
            {tangentMode=RCTM_User, time=0.9216, value=8}
            <at 1 8>
           ]}
          GreenCurve {keys [
            <at 0.1481 28.1816>
            {
             arriveTangent=-26.0882
             arriveTangentWeight=0.0167
             interpMode=RCIM_Cubic
             leaveTangent=-26.0882
             tangentMode=RCTM_User
             time=0.198
             value=12
            }
            {tangentMode=RCTM_User, time=0.9216, value=8}
            <at 1 8>
           ]}
          RedCurve {keys [
            <at 0.1481 7.2711>
            {arriveTangent=0.9422, arriveTangentWeight=0.0167, interpMode=RCIM_Cubic, leaveTangent=0.9422, tangentMode=RCTM_User, time=0.198, value=12}
            {tangentMode=RCTM_User, time=0.9216, value=8}
            <at 1 8>
           ]}
         }}
        CurveIndex=&Subtract_Float(
         A=&Multiply_Float(A=$Particles.Age, B=&Divide_Float(A=1, B=$Particles.Lifetime))
         B=&Multiply_Float(
          A=&SelectFloatFromArray(
           Direct Array Index=&ModuloInt(A=&ReturnParticleIDAttributes(), B=15)
           Float Selection Array=$System.Fruit Explode Times
          )
          B=$Particles.Lifetime
         )
        )
       )
       Scale Factor=&MakeVector4_FromLinearColor(LinearColor=&LerpLinearColors(StartColor=<rgba 0.0019 0.5703 0.6979 1>, EndColor=<rgba 0.002 0.724 0.6397 1>, LerpFactor=$Particles.MaterialRandom))
      ))
     False Vector=<1 1 1>
    )
    ScaleA=true
    ScaleRGB=true
   )
   /SetVariables(Particles.Velocity=&Add_Vector(
     A=$Particles.Velocity
     B=&Multiply_VectorByFloat(
      Vector=$Particles.accumulatedVelocity
      Float=&MakeCustomFloatFromBool(Bool=$Particles.DoOnce.Execute, True Float=0.75, False Float=0)
     )
    )) {isSetParameters=true, module=SetVariables_#9}
  ]
  rendererClasses [&NiagaraSpriteRendererProperties]
  renderers [
   {rendererClass=&NiagaraSpriteRendererProperties}
   {rendererClass=&NiagaraSpriteRendererProperties, rendererIndex=1}
  ]
 }
 FruitSplashesAndDrops001 {
  simTarget=GPUComputeSim
  emitterSpawnScript [
   /SetVariables(Emitter.Splashes Per Bulb:NiagaraInt32=&MakeCustomIntFromBool(
     Bool=&IsPlatformSetActive(New Platform Set=DataInterface() {propertyValues {Platforms {qualityLevelMask=-31}}})
     True INT=1
     False INT=&MakeCustomIntFromBool(
      Bool=&IsPlatformSetActive(New Platform Set=DataInterface() {propertyValues {Platforms {qualityLevelMask=-30}}})
      True INT=1
      False INT=&MakeCustomIntFromBool(
       Bool=&IsPlatformSetActive(New Platform Set=DataInterface() {propertyValues {Platforms {qualityLevelMask=-28}}})
       True INT=&RandomRangeInt(Minimum=1, Maximum=2)
       False INT=&RandomRangeInt(Minimum=2, Maximum=4)
      )
     )
    )) {isSetParameters=true, module=SetVariables_#1}
   /SetVariables(Emitter.Drops Per Bulb:NiagaraInt32=&MakeCustomIntFromBool(
     Bool=&IsPlatformSetActive(New Platform Set=DataInterface() {propertyValues {Platforms {qualityLevelMask=-31}}})
     True INT=1
     False INT=&MakeCustomIntFromBool(
      Bool=&IsPlatformSetActive(New Platform Set=DataInterface() {propertyValues {Platforms {qualityLevelMask=-30}}})
      True INT=&RandomRangeInt(Minimum=1, Maximum=3)
      False INT=&MakeCustomIntFromBool(
       Bool=&IsPlatformSetActive(New Platform Set=DataInterface() {propertyValues {Platforms {qualityLevelMask=-28}}})
       True INT=&RandomRangeInt(Minimum=2, Maximum=4)
       False INT=&RandomRangeInt(Minimum=3, Maximum=5)
      )
     )
    )) {isSetParameters=true, module=SetVariables_#2}
   /SetVariables(Emitter.Goop Spawn Scale=0.5) {enabled=false, isSetParameters=true, module=SetVariables_#3}
   /SetVariables(Emitter.Drop Spawn Scale=0.5) {enabled=false, isSetParameters=true, module=SetVariables_#4}
  ]
  emitterUpdateScript [
   &EmitterState(
    Enable Distance Culling=true
    Life Cycle Mode=::NewEnumerator1
    Loop Behavior=::NewEnumerator1
    Loop Duration=1
    MaxDistance=2800
    Module.bUseMaxDistance=true
    Module.bUseMinDistance=false
    Scalability Mode=::NewEnumerator1
    UseLoopDelay=false
   )
   &SpawnBurst_Instantaneous(
    Spawn Count=&Multiply_Int(A=$Emitter.Splashes Per Bulb, B=15)
    Spawn Time=&MakeCustomFloatFromBool(Bool=$User.WasHitByVeh, True Float=0, False Float=0.05)
    Use Spawn Probability=false
    UseLoopCountLimit=false
   )
   &SpawnBurst_Instantaneous(
    Spawn Count=&Multiply_Int(A=$Emitter.Drops Per Bulb, B=15)
    Spawn Group=15
    Spawn Probability=&MakeCustomFloatFromBool(
     Bool=&IsPlatformSetActive(New Platform Set=DataInterface() {propertyValues {Platforms {qualityLevelMask=-29}}})
     True Float=0
     False Float=1
    )
    Spawn Time=&MakeCustomFloatFromBool(Bool=$User.WasHitByVeh, True Float=0, False Float=0.05)
    Use Spawn Probability=true
    UseLoopCountLimit=false
   ) {module=SpawnBurst_Instantaneous015}
  ]
  particleSpawnScript [
   /SetVariables(Particles.bIsSplashes:NiagaraBool=&MaskBoolBySpawnGroup(Masked Bool:NiagaraBool=true, Passthrough Bool:NiagaraBool=false, Spawn Group Mask:NiagaraInt32=0)) {isSetParameters=true, module=SetVariables_#5}
   &InitializeParticle(
    Color=&MakeCustomLinearColorFromBool(
     Bool=$Particles.bIsSplashes
     True Color=&RandomRangeLinearColor(Minimum=<rgba 4 4 4 1>, Maximum=<rgba 0.7835 0.7835 0.7835 0.6>)
     False Color=&RandomRangeLinearColor(Minimum=<rgba 3 3 3 1>, Maximum=<rgba 4 4 4 1>)
    )
    Lifetime=&Add_Float(
     A=&MakeCustomFloatFromBool(
      Bool=$Particles.bIsSplashes
      True Float=&RandomRangeFloat(Minimum=0.85, Maximum=1.5)
      False Float=&RandomRangeFloat(Minimum=0.5, Maximum=0.75)
     )
     B=&SelectFloatFromArray(
      Direct Array Index=&ModuloInt(A=&ReturnParticleIDAttributes(), B=15)
      Float Selection Array=$System.Fruit Explode Times
     )
    )
    Mass Max=2
    Mass Mode=::NewEnumerator3
    Sprite Rotation Mode=::NewEnumerator2
    Sprite Size Mode=::NewEnumerator3
    Sprite UV Mode=::NewEnumerator3
    Uniform Sprite Size=&MakeCustomFloatFromBool(
     Bool=$Particles.bIsSplashes
     True Float=&RandomRangeFloat(Minimum=50, Maximum=75)
     False Float=&RandomRangeFloat(Minimum=7, Maximum=10)
    )
    UsePositionOffset=false
    Write Material Random=true
   )
   &ShapeLocation(
    MaskSpawnGroup=false
    Non Uniform Scale=&MakeCustomVectorFromBool(Bool=$Particles.bIsSplashes, True Vector=<0 0 0>, False Vector=<1 1 2>)
    Offset=&MakeCustomVectorFromBool(Bool=$Particles.bIsSplashes, True Vector=<0 0 0>, False Vector=<0 0 50>)
    Offset Mode=::NewEnumerator0
    Override Seed=false
    Sphere Radius=&MakeCustomFloatFromBool(Bool=$Particles.bIsSplashes, True Float=0, False Float=5)
    Sphere Surface Distribution=1
    UseOwnerScale=true
   )
   &AddVelocityFromPoint(
    Module.Use Falloff Distance=false
    Origin Offset=<0 0 52>
    Velocity Strength=&MakeCustomFloatFromBool(Bool=$Particles.bIsSplashes, True Float=0, False Float=&UniformRangedFloat(Minimum=500, Maximum=200))
   )
   &AddVelocity(Velocity=&MakeCustomVectorFromBool(Bool=$Particles.bIsSplashes, True Vector=<0 0 0>, False Vector=<0 0 400>))
   &StaticMeshLocation(
    Interpolate Mesh Sampling=false
    MaskSpawnGroup=false
    Mesh Sampling Type:ENiagara_StaticSamplingMode=::NewEnumerator1
    SamplePosOffset=true
    Sampled Position Offset:Vector3f=<0 0 -7>
    Sampled Position Offset Coordinate Space:ENiagaraSkelMeshTransforms=::NewEnumerator1
    Socket Index:NiagaraInt32=&ModuloInt(A=&ReturnParticleIDAttributes(), B=15)
    Socket Sampling Mode:ENiagara_StaticSocketMode=::NewEnumerator2
    Static Mesh:NiagaraDataInterfaceStaticMesh=$System.Static Mesh
   )
   &CameraOffset(Camera Offset Amount=&MakeCustomFloatFromBool(Bool=$Particles.bIsSplashes, True Float=15, False Float=0))
   &AddVelocity(
    Cone Angle=60
    Cone Axis=&ReturnMeshOrientationAxis(Mesh Orientation Axis=::XAxis, Mesh Orientation Quat=$Particles.StaticMeshLocation.SampledRotation)
    Module.UseConeFalloff=false
    Module.UseVelDistribution=false
    Override Seed=false
    Velocity Mode=::NewEnumerator2
    Velocity Speed=&MakeCustomFloatFromBool(Bool=$Particles.bIsSplashes, True Float=600, False Float=0)
   ) {module=AddVelocity003}
   &DynamicMaterialParameters(
    Index 0 Param 1=&RandomRangeFloat(Minimum=1, Maximum=2)
    Index 0 Param 2=0.5
    Index 0 Param 3=&RandomRangeFloat(Minimum=15, Maximum=50)
    Index 0 Param 4=&RandomRangeFloat(Minimum=2, Maximum=10)
    Index 1 Param 1=&RandomRangeFloat(Minimum=0, Maximum=10)
    Module.Param0WriteEnabled=true
    Module.Param0WriteEnabled1=true
    Module.Param1WriteEnabled=true
    Module.Param1WriteEnabled1=false
    Module.Param2WriteEnabled=true
    Module.Param2WriteEnabled1=false
    Module.Param3WriteEnabled=true
    Module.Param3WriteEnabled1=false
    Write Parameter Index 1=true
   )
   /SetVariables(Particles.VisibilityTag=&MakeCustomIntFromBool(Bool=$Particles.bIsSplashes, True INT=0, False INT=1)) {isSetParameters=true, module=SetVariables_#6}
   /SetVariables(Particles.accumulatedVelocity:Vector3f=$Particles.Velocity) {isSetParameters=true, module=SetVariables_#7}
   /SetVariables() {isSetParameters=true, module=SetVariables_#8}
  ]
  particleUpdateScript [
   &ParticleState()
   &ScaleSpriteSize(
    Scale Sprite Size Mode=::NewEnumerator0
    Uniform Scale Factor=&MakeCustomFloatFromBool(
     Bool=&SetBoolByFloatComparison(
      A=$Particles.Age
      B=&SelectFloatFromArray(
       Direct Array Index=&ModuloInt(A=&ReturnParticleIDAttributes(), B=15)
       Float Selection Array=$System.Fruit Explode Times
      )
     )
     True Float=1
     False Float=0
    )
   ) {module=ScaleSpriteSize001}
   &SpriteRotationRate(Rotation Rate=&MakeCustomFloatFromBool(
     Bool=$Particles.bIsSplashes
     True Float=&Multiply_Float(
      A=&Lerp_Float(
       A=180
       B=45
       Alpha=&RemapRange(Input Value To Remap=$Particles.Mass, Input Min=0.75, Input Max=2, Output Min=0, Output Max=1)
      )
      B=&MakeFloatFromInt(Int=&RandomRangeInt(Minimum=-1, Maximum=1))
     )
     False Float=0
    ))
   &DoOnce(Trigger Condition:NiagaraBool=&SetBoolByFloatComparison(
     A=$Emitter.Age
     B=&SelectFloatFromArray(
      Direct Array Index=&ModuloInt(A=&ReturnParticleIDAttributes(), B=15)
      Direct Array Mode=::NewEnumerator1
      Float Selection Array=$System.Fruit Explode Times
     )
    ))
   &Drag(
    Drag=&MakeCustomFloatFromBool(
     Bool=$Particles.bIsSplashes
     True Float=&FloatFromCurve(
      FloatCurve=DataInterface() {propertyValues {Curve {keys [<at 0 2 RCIM_Cubic>,<at 0.5 0>]}}}
      Scale Curve=&RandomRangeFloat(Minimum=2, Maximum=2.5)
     )
     False Float=&FloatFromCurve(
      FloatCurve=DataInterface() {propertyValues {Curve {keys [<at 0 1 RCIM_Cubic>, {arriveTangent=-1.4297, interpMode=RCIM_Cubic, leaveTangent=-1.4297, tangentMode=RCTM_User, time=1, value=0}]}}}
      Scale Curve=&UniformRangedFloat(Minimum=10, Maximum=8)
     )
    )
    Use Linear Drag=true
    Use Rotational Drag=true
   )
   &GravityForce(Gravity=&Multiply_VectorByFloat(
     Vector=&MakeCustomVectorFromBool(Bool=$Particles.bIsSplashes, True Vector=<0 0 -980>, False Vector=<0 0 -1470>)
     Float=&MakeCustomFloatFromBool(Bool=$Particles.DoOnce.TriggeredOnce, True Float=1, False Float=0)
    ))
   &SolveForcesAndVelocity(Clamp Velocity=false, Limit Acceleration=false)
   &ScaleSpriteSizeBySpeed(
    Max Scale Factor=&MakeCustomVector2DFromBool(Bool=$Particles.bIsSplashes, True Vector2D=<1 1>, False Vector2D=<1.2 4>)
    Min Scale Factor=&MakeCustomVector2DFromBool(Bool=$Particles.bIsSplashes, True Vector2D=<1 1>, False Vector2D=<0.25 0.2>)
   )
   &ScaleSpriteSize(
    Scale Sprite Size Mode=::NewEnumerator0
    Uniform Scale Factor=&MakeCustomFloatFromBool(
     Bool=$Particles.bIsSplashes
     True Float=&FloatFromCurve(
      FloatCurve=DataInterface() {propertyValues {Curve {keys [
          <at 0 0 RCIM_Cubic>
          {arriveTangent=1.8917, interpMode=RCIM_Cubic, leaveTangent=1.8917, tangentMode=RCTM_User, time=0.001, value=0.35}
          {interpMode=RCIM_Cubic, leaveTangent=-0.5754, time=1, value=2}
         ]}}}
      CurveIndex=&Lerp_Float(
       A=0
       B=1
       Alpha=&self:inverseLerp(
        A:NiagaraFloat=&SelectFloatFromArray(
         Direct Array Index=&ModuloInt(A=&ReturnParticleIDAttributes(), B=15)
         Float Selection Array=$System.Fruit Explode Times
        )
        B:NiagaraFloat=$Particles.Lifetime
        Clamp:NiagaraBool=true
        V:NiagaraFloat=$Particles.Age
       )
      )
     )
     False Float=1
    )
   )
   &ScaleColor(
    Scale Alpha=&MakeCustomFloatFromBool(
     Bool=$Particles.bIsSplashes
     True Float=&FloatFromCurve(
      FloatCurve=DataInterface() {propertyValues {Curve {keys [
          <at 0 -1>
          {arriveTangentWeight=0.0001, interpMode=RCIM_Cubic, leaveTangentWeight=0.3333, tangentMode=RCTM_User, time=0.0002, value=0.5}
          {arriveTangent=-2.4995, arriveTangentWeight=0.8972, interpMode=RCIM_Cubic, leaveTangent=-2.4995, tangentMode=RCTM_User, time=1, value=-1}
         ]}}}
      CurveIndex=&Lerp_Float(
       A=0
       B=1
       Alpha=&self:inverseLerp(
        A:NiagaraFloat=&SelectFloatFromArray(
         Direct Array Index=&ModuloInt(A=&ReturnParticleIDAttributes(), B=15)
         Float Selection Array=$System.Fruit Explode Times
        )
        B:NiagaraFloat=$Particles.Lifetime
        Clamp:NiagaraBool=true
        V:NiagaraFloat=$Particles.Age
       )
      )
     )
     False Float=&FloatFromCurve(FloatCurve=DataInterface() {propertyValues {Curve {keys [<at 0 1 RCIM_Cubic>, {arriveTangent=-3.0079, interpMode=RCIM_Cubic, leaveTangent=-3.0079, tangentMode=RCTM_User, time=1, value=0}]}}})
    )
    Scale RGB=&MakeCustomVectorFromBool(
     Bool=$Particles.bIsSplashes
     True Vector=&MakeVectorFromLinearColorRGB(Color=&MultiplyLinearColor(
       Linear Color=&ColorFromCurve(
        ColorCurve=DataInterface() {propertyValues {
          AlphaCurve {keys []}
          BlueCurve {keys [
            <at 0.1481 40>
            {
             arriveTangent=-41.3655
             arriveTangentWeight=0.0167
             interpMode=RCIM_Cubic
             leaveTangent=-41.3655
             tangentMode=RCTM_User
             time=0.198
             value=12
            }
            {tangentMode=RCTM_User, time=0.9216, value=8}
            <at 1 8>
           ]}
          GreenCurve {keys [
            <at 0.1481 28.1816>
            {
             arriveTangent=-26.0882
             arriveTangentWeight=0.0167
             interpMode=RCIM_Cubic
             leaveTangent=-26.0882
             tangentMode=RCTM_User
             time=0.198
             value=12
            }
            {tangentMode=RCTM_User, time=0.9216, value=8}
            <at 1 8>
           ]}
          RedCurve {keys [
            <at 0.1481 7.2711>
            {arriveTangent=0.9422, arriveTangentWeight=0.0167, interpMode=RCIM_Cubic, leaveTangent=0.9422, tangentMode=RCTM_User, time=0.198, value=12}
            {tangentMode=RCTM_User, time=0.9216, value=8}
            <at 1 8>
           ]}
         }}
        CurveIndex=&Lerp_Float(
         A=0
         B=1
         Alpha=&self:inverseLerp(
          A:NiagaraFloat=&SelectFloatFromArray(
           Direct Array Index=&ModuloInt(A=&ReturnParticleIDAttributes(), B=15)
           Float Selection Array=$System.Fruit Explode Times
          )
          B:NiagaraFloat=$Particles.Lifetime
          Clamp:NiagaraBool=true
          V:NiagaraFloat=$Particles.Age
         )
        )
       )
       Scale Factor=&MakeVector4_FromLinearColor(LinearColor=&LerpLinearColors(StartColor=<rgba 0.0019 0.5703 0.6979 1>, EndColor=<rgba 0.002 0.724 0.6397 1>, LerpFactor=$Particles.MaterialRandom))
      ))
     False Vector=<1 1 1>
    )
    ScaleA=true
    ScaleRGB=true
   )
   /SetVariables(Particles.Velocity=&Add_Vector(
     A=$Particles.Velocity
     B=&Multiply_VectorByFloat(
      Vector=$Particles.accumulatedVelocity
      Float=&MakeCustomFloatFromBool(Bool=$Particles.DoOnce.Execute, True Float=1, False Float=0)
     )
    )) {isSetParameters=true, module=SetVariables_#9}
  ]
  rendererClasses [&NiagaraSpriteRendererProperties]
  renderers [
   {rendererClass=&NiagaraSpriteRendererProperties}
   {rendererClass=&NiagaraSpriteRendererProperties, rendererIndex=1}
  ]
 }
 FruitSplashesMergeOld {
  bEnabled=false
  simTarget=GPUComputeSim
  emitterSpawnScript [
   /SetVariables(Emitter.Particles Per Bulb:NiagaraInt32=&MakeCustomIntFromBool(
     Bool=&IsPlatformSetActive(New Platform Set=DataInterface() {propertyValues {Platforms {qualityLevelMask=-31}}})
     True INT=1
     False INT=&MakeCustomIntFromBool(
      Bool=&IsPlatformSetActive(New Platform Set=DataInterface() {propertyValues {Platforms {qualityLevelMask=-30}}})
      True INT=1
      False INT=&MakeCustomIntFromBool(
       Bool=&IsPlatformSetActive(New Platform Set=DataInterface() {propertyValues {Platforms {qualityLevelMask=-28}}})
       True INT=&RandomRangeInt(Minimum=1, Maximum=2)
       False INT=&RandomRangeInt(Minimum=3, Maximum=4)
      )
     )
    )) {isSetParameters=true, module=SetVariables_#1}
   /SetVariables(Emitter.Goop Spawn Scale=0.5) {isSetParameters=true, module=SetVariables_#3}
   /SetVariables(Emitter.Drop Spawn Scale=0.5) {isSetParameters=true, module=SetVariables_#4}
  ]
  emitterUpdateScript [
   &EmitterState(
    Enable Distance Culling=true
    Life Cycle Mode=::NewEnumerator1
    Loop Behavior=::NewEnumerator1
    Loop Duration=1
    MaxDistance=2800
    Module.bUseMaxDistance=true
    Module.bUseMinDistance=false
    Scalability Mode=::NewEnumerator1
    UseLoopDelay=true
   )
   &SpawnBurst_Instantaneous(
    Spawn Count=&self:In_By_FLoat(int:NiagaraInt32=$Emitter.Particles Per Bulb, float:NiagaraFloat=1)
    Spawn Time=&MakeCustomFloatFromBool(Bool=$User.WasHitByVeh, True Float=0, False Float=0.05)
    Use Spawn Probability=false
    UseLoopCountLimit=false
   )
   &SpawnBurst_Instantaneous(
    Spawn Count=&self:In_By_FLoat(int:NiagaraInt32=$Emitter.Particles Per Bulb, float:NiagaraFloat=$Emitter.Goop Spawn Scale)
    Spawn Group=1
    Spawn Time=&MakeCustomFloatFromBool(Bool=$User.WasHitByVeh, True Float=0, False Float=0.1)
    Use Spawn Probability=false
    UseLoopCountLimit=false
   ) {module=SpawnBurst_Instantaneous001}
   &SpawnBurst_Instantaneous(
    Spawn Count=&self:In_By_FLoat(int:NiagaraInt32=$Emitter.Particles Per Bulb, float:NiagaraFloat=1)
    Spawn Group=2
    Spawn Time=&MakeCustomFloatFromBool(Bool=$User.WasHitByVeh, True Float=0, False Float=0.125)
    Use Spawn Probability=false
    UseLoopCountLimit=false
   ) {module=SpawnBurst_Instantaneous002}
   &SpawnBurst_Instantaneous(
    Spawn Count=&self:In_By_FLoat(int:NiagaraInt32=$Emitter.Particles Per Bulb, float:NiagaraFloat=$Emitter.Goop Spawn Scale)
    Spawn Group=3
    Spawn Time=&MakeCustomFloatFromBool(Bool=$User.WasHitByVeh, True Float=0, False Float=0.125)
    Use Spawn Probability=false
    UseLoopCountLimit=false
   ) {module=SpawnBurst_Instantaneous003}
   &SpawnBurst_Instantaneous(
    Spawn Count=&self:In_By_FLoat(int:NiagaraInt32=$Emitter.Particles Per Bulb, float:NiagaraFloat=1)
    Spawn Group=4
    Spawn Time=&MakeCustomFloatFromBool(Bool=$User.WasHitByVeh, True Float=0, False Float=0.125)
    Use Spawn Probability=false
    UseLoopCountLimit=false
   ) {module=SpawnBurst_Instantaneous004}
   &SpawnBurst_Instantaneous(
    Spawn Count=&self:In_By_FLoat(int:NiagaraInt32=$Emitter.Particles Per Bulb, float:NiagaraFloat=$Emitter.Goop Spawn Scale)
    Spawn Group=5
    Spawn Time=&MakeCustomFloatFromBool(Bool=$User.WasHitByVeh, True Float=0, False Float=0.175)
    Use Spawn Probability=false
    UseLoopCountLimit=false
   ) {module=SpawnBurst_Instantaneous005}
   &SpawnBurst_Instantaneous(
    Spawn Count=&self:In_By_FLoat(int:NiagaraInt32=$Emitter.Particles Per Bulb, float:NiagaraFloat=1)
    Spawn Group=6
    Spawn Time=&MakeCustomFloatFromBool(Bool=$User.WasHitByVeh, True Float=0, False Float=0.18)
    Use Spawn Probability=false
    UseLoopCountLimit=false
   ) {module=SpawnBurst_Instantaneous006}
   &SpawnBurst_Instantaneous(
    Spawn Count=&self:In_By_FLoat(int:NiagaraInt32=$Emitter.Particles Per Bulb, float:NiagaraFloat=$Emitter.Goop Spawn Scale)
    Spawn Group=7
    Spawn Time=&MakeCustomFloatFromBool(Bool=$User.WasHitByVeh, True Float=0, False Float=0.22)
    Use Spawn Probability=false
    UseLoopCountLimit=false
   ) {module=SpawnBurst_Instantaneous007}
   &SpawnBurst_Instantaneous(
    Spawn Count=&self:In_By_FLoat(int:NiagaraInt32=$Emitter.Particles Per Bulb, float:NiagaraFloat=1)
    Spawn Group=8
    Spawn Time=&MakeCustomFloatFromBool(Bool=$User.WasHitByVeh, True Float=0, False Float=0.235)
    Use Spawn Probability=false
    UseLoopCountLimit=false
   ) {module=SpawnBurst_Instantaneous008}
   &SpawnBurst_Instantaneous(
    Spawn Count=&self:In_By_FLoat(int:NiagaraInt32=$Emitter.Particles Per Bulb, float:NiagaraFloat=$Emitter.Goop Spawn Scale)
    Spawn Group=9
    Spawn Time=&MakeCustomFloatFromBool(Bool=$User.WasHitByVeh, True Float=0, False Float=0.23)
    Use Spawn Probability=false
    UseLoopCountLimit=false
   ) {module=SpawnBurst_Instantaneous009}
   &SpawnBurst_Instantaneous(
    Spawn Count=&self:In_By_FLoat(int:NiagaraInt32=$Emitter.Particles Per Bulb, float:NiagaraFloat=1)
    Spawn Group=10
    Spawn Time=&MakeCustomFloatFromBool(Bool=$User.WasHitByVeh, True Float=0, False Float=0.23)
    Use Spawn Probability=false
    UseLoopCountLimit=false
   ) {module=SpawnBurst_Instantaneous010}
   &SpawnBurst_Instantaneous(
    Spawn Count=&self:In_By_FLoat(int:NiagaraInt32=$Emitter.Particles Per Bulb, float:NiagaraFloat=$Emitter.Goop Spawn Scale)
    Spawn Group=11
    Spawn Time=&MakeCustomFloatFromBool(Bool=$User.WasHitByVeh, True Float=0, False Float=0.26)
    Use Spawn Probability=false
    UseLoopCountLimit=false
   ) {module=SpawnBurst_Instantaneous011}
   &SpawnBurst_Instantaneous(
    Spawn Count=&self:In_By_FLoat(int:NiagaraInt32=$Emitter.Particles Per Bulb, float:NiagaraFloat=1)
    Spawn Group=12
    Spawn Time=&MakeCustomFloatFromBool(Bool=$User.WasHitByVeh, True Float=0, False Float=0.28)
    Use Spawn Probability=false
    UseLoopCountLimit=false
   ) {module=SpawnBurst_Instantaneous012}
   &SpawnBurst_Instantaneous(
    Spawn Count=&self:In_By_FLoat(int:NiagaraInt32=$Emitter.Particles Per Bulb, float:NiagaraFloat=$Emitter.Goop Spawn Scale)
    Spawn Group=13
    Spawn Time=&MakeCustomFloatFromBool(Bool=$User.WasHitByVeh, True Float=0, False Float=0.3)
    Use Spawn Probability=false
    UseLoopCountLimit=false
   ) {module=SpawnBurst_Instantaneous013}
   &SpawnBurst_Instantaneous(
    Spawn Count=&self:In_By_FLoat(int:NiagaraInt32=$Emitter.Particles Per Bulb, float:NiagaraFloat=1)
    Spawn Group=14
    Spawn Time=&MakeCustomFloatFromBool(Bool=$User.WasHitByVeh, True Float=0, False Float=0.325)
    Use Spawn Probability=false
    UseLoopCountLimit=false
   ) {module=SpawnBurst_Instantaneous014}
   &SpawnBurst_Instantaneous(
    Spawn Count=&self:In_By_FLoat(int:NiagaraInt32=$Emitter.Particles Per Bulb, float:NiagaraFloat=$Emitter.Drop Spawn Scale)
    Spawn Group=15
    Spawn Probability=&MakeCustomFloatFromBool(
     Bool=&IsPlatformSetActive(New Platform Set=DataInterface() {propertyValues {Platforms {qualityLevelMask=-29}}})
     True Float=0
     False Float=1
    )
    Spawn Time=&MakeCustomFloatFromBool(Bool=$User.WasHitByVeh, True Float=0, False Float=0.05)
    Use Spawn Probability=true
    UseLoopCountLimit=false
   ) {module=SpawnBurst_Instantaneous015}
   &SpawnBurst_Instantaneous(
    Spawn Count=$Emitter.Particles Per Bulb
    Spawn Group=16
    Spawn Probability=&MakeCustomFloatFromBool(
     Bool=&IsPlatformSetActive(New Platform Set=DataInterface() {propertyValues {Platforms {qualityLevelMask=-29}}})
     True Float=0
     False Float=1
    )
    Spawn Time=&MakeCustomFloatFromBool(Bool=$User.WasHitByVeh, True Float=0, False Float=0.1)
    Use Spawn Probability=true
    UseLoopCountLimit=false
   ) {module=SpawnBurst_Instantaneous016}
   &SpawnBurst_Instantaneous(
    Spawn Count=&self:In_By_FLoat(int:NiagaraInt32=$Emitter.Particles Per Bulb, float:NiagaraFloat=$Emitter.Drop Spawn Scale)
    Spawn Group=17
    Spawn Probability=&MakeCustomFloatFromBool(
     Bool=&IsPlatformSetActive(New Platform Set=DataInterface() {propertyValues {Platforms {qualityLevelMask=-29}}})
     True Float=0
     False Float=1
    )
    Spawn Time=&MakeCustomFloatFromBool(Bool=$User.WasHitByVeh, True Float=0, False Float=0.125)
    Use Spawn Probability=true
    UseLoopCountLimit=false
   ) {module=SpawnBurst_Instantaneous017}
   &SpawnBurst_Instantaneous(
    Spawn Count=$Emitter.Particles Per Bulb
    Spawn Group=18
    Spawn Probability=&MakeCustomFloatFromBool(
     Bool=&IsPlatformSetActive(New Platform Set=DataInterface() {propertyValues {Platforms {qualityLevelMask=-29}}})
     True Float=0
     False Float=1
    )
    Spawn Time=&MakeCustomFloatFromBool(Bool=$User.WasHitByVeh, True Float=0, False Float=0.125)
    Use Spawn Probability=true
    UseLoopCountLimit=false
   ) {module=SpawnBurst_Instantaneous018}
   &SpawnBurst_Instantaneous(
    Spawn Count=&self:In_By_FLoat(int:NiagaraInt32=$Emitter.Particles Per Bulb, float:NiagaraFloat=$Emitter.Drop Spawn Scale)
    Spawn Group=19
    Spawn Probability=&MakeCustomFloatFromBool(
     Bool=&IsPlatformSetActive(New Platform Set=DataInterface() {propertyValues {Platforms {qualityLevelMask=-29}}})
     True Float=0
     False Float=1
    )
    Spawn Time=&MakeCustomFloatFromBool(Bool=$User.WasHitByVeh, True Float=0, False Float=0.125)
    Use Spawn Probability=true
    UseLoopCountLimit=false
   ) {module=SpawnBurst_Instantaneous019}
   &SpawnBurst_Instantaneous(
    Spawn Count=$Emitter.Particles Per Bulb
    Spawn Group=20
    Spawn Probability=&MakeCustomFloatFromBool(
     Bool=&IsPlatformSetActive(New Platform Set=DataInterface() {propertyValues {Platforms {qualityLevelMask=-29}}})
     True Float=0
     False Float=1
    )
    Spawn Time=&MakeCustomFloatFromBool(Bool=$User.WasHitByVeh, True Float=0, False Float=0.175)
    Use Spawn Probability=true
    UseLoopCountLimit=false
   ) {module=SpawnBurst_Instantaneous020}
   &SpawnBurst_Instantaneous(
    Spawn Count=&self:In_By_FLoat(int:NiagaraInt32=$Emitter.Particles Per Bulb, float:NiagaraFloat=$Emitter.Drop Spawn Scale)
    Spawn Group=21
    Spawn Probability=&MakeCustomFloatFromBool(
     Bool=&IsPlatformSetActive(New Platform Set=DataInterface() {propertyValues {Platforms {qualityLevelMask=-29}}})
     True Float=0
     False Float=1
    )
    Spawn Time=&MakeCustomFloatFromBool(Bool=$User.WasHitByVeh, True Float=0, False Float=0.18)
    Use Spawn Probability=true
    UseLoopCountLimit=false
   ) {module=SpawnBurst_Instantaneous021}
   &SpawnBurst_Instantaneous(
    Spawn Count=$Emitter.Particles Per Bulb
    Spawn Group=22
    Spawn Probability=&MakeCustomFloatFromBool(
     Bool=&IsPlatformSetActive(New Platform Set=DataInterface() {propertyValues {Platforms {qualityLevelMask=-29}}})
     True Float=0
     False Float=1
    )
    Spawn Time=&MakeCustomFloatFromBool(Bool=$User.WasHitByVeh, True Float=0, False Float=0.22)
    Use Spawn Probability=true
    UseLoopCountLimit=false
   ) {module=SpawnBurst_Instantaneous022}
   &SpawnBurst_Instantaneous(
    Spawn Count=&self:In_By_FLoat(int:NiagaraInt32=$Emitter.Particles Per Bulb, float:NiagaraFloat=$Emitter.Drop Spawn Scale)
    Spawn Group=23
    Spawn Probability=&MakeCustomFloatFromBool(
     Bool=&IsPlatformSetActive(New Platform Set=DataInterface() {propertyValues {Platforms {qualityLevelMask=-29}}})
     True Float=0
     False Float=1
    )
    Spawn Time=&MakeCustomFloatFromBool(Bool=$User.WasHitByVeh, True Float=0, False Float=0.235)
    Use Spawn Probability=true
    UseLoopCountLimit=false
   ) {module=SpawnBurst_Instantaneous023}
   &SpawnBurst_Instantaneous(
    Spawn Count=$Emitter.Particles Per Bulb
    Spawn Group=24
    Spawn Probability=&MakeCustomFloatFromBool(
     Bool=&IsPlatformSetActive(New Platform Set=DataInterface() {propertyValues {Platforms {qualityLevelMask=-29}}})
     True Float=0
     False Float=1
    )
    Spawn Time=&MakeCustomFloatFromBool(Bool=$User.WasHitByVeh, True Float=0, False Float=0.23)
    Use Spawn Probability=true
    UseLoopCountLimit=false
   ) {module=SpawnBurst_Instantaneous024}
   &SpawnBurst_Instantaneous(
    Spawn Count=&self:In_By_FLoat(int:NiagaraInt32=$Emitter.Particles Per Bulb, float:NiagaraFloat=$Emitter.Drop Spawn Scale)
    Spawn Group=25
    Spawn Probability=&MakeCustomFloatFromBool(
     Bool=&IsPlatformSetActive(New Platform Set=DataInterface() {propertyValues {Platforms {qualityLevelMask=-29}}})
     True Float=0
     False Float=1
    )
    Spawn Time=&MakeCustomFloatFromBool(Bool=$User.WasHitByVeh, True Float=0, False Float=0.23)
    Use Spawn Probability=true
    UseLoopCountLimit=false
   ) {module=SpawnBurst_Instantaneous025}
   &SpawnBurst_Instantaneous(
    Spawn Count=$Emitter.Particles Per Bulb
    Spawn Group=26
    Spawn Probability=&MakeCustomFloatFromBool(
     Bool=&IsPlatformSetActive(New Platform Set=DataInterface() {propertyValues {Platforms {qualityLevelMask=-29}}})
     True Float=0
     False Float=1
    )
    Spawn Time=&MakeCustomFloatFromBool(Bool=$User.WasHitByVeh, True Float=0, False Float=0.26)
    Use Spawn Probability=true
    UseLoopCountLimit=false
   ) {module=SpawnBurst_Instantaneous026}
   &SpawnBurst_Instantaneous(
    Spawn Count=&self:In_By_FLoat(int:NiagaraInt32=$Emitter.Particles Per Bulb, float:NiagaraFloat=$Emitter.Drop Spawn Scale)
    Spawn Group=27
    Spawn Probability=&MakeCustomFloatFromBool(
     Bool=&IsPlatformSetActive(New Platform Set=DataInterface() {propertyValues {Platforms {qualityLevelMask=-29}}})
     True Float=0
     False Float=1
    )
    Spawn Time=&MakeCustomFloatFromBool(Bool=$User.WasHitByVeh, True Float=0, False Float=0.28)
    Use Spawn Probability=true
    UseLoopCountLimit=false
   ) {module=SpawnBurst_Instantaneous027}
   &SpawnBurst_Instantaneous(
    Spawn Count=$Emitter.Particles Per Bulb
    Spawn Group=28
    Spawn Probability=&MakeCustomFloatFromBool(
     Bool=&IsPlatformSetActive(New Platform Set=DataInterface() {propertyValues {Platforms {qualityLevelMask=-29}}})
     True Float=0
     False Float=1
    )
    Spawn Time=&MakeCustomFloatFromBool(Bool=$User.WasHitByVeh, True Float=0, False Float=0.3)
    Use Spawn Probability=true
    UseLoopCountLimit=false
   ) {module=SpawnBurst_Instantaneous028}
   &SpawnBurst_Instantaneous(
    Spawn Count=&self:In_By_FLoat(int:NiagaraInt32=$Emitter.Particles Per Bulb, float:NiagaraFloat=$Emitter.Drop Spawn Scale)
    Spawn Group=29
    Spawn Probability=&MakeCustomFloatFromBool(
     Bool=&IsPlatformSetActive(New Platform Set=DataInterface() {propertyValues {Platforms {qualityLevelMask=-29}}})
     True Float=0
     False Float=1
    )
    Spawn Time=&MakeCustomFloatFromBool(Bool=$User.WasHitByVeh, True Float=0, False Float=0.325)
    Use Spawn Probability=true
    UseLoopCountLimit=false
   ) {module=SpawnBurst_Instantaneous029}
  ]
  particleSpawnScript [
   /SetVariables(Particles.bIsSplashes:NiagaraBool=&SetBoolByIntComparison(A=&SpawnGroup(), B=15, Comparison Type=::NewEnumerator6)) {isSetParameters=true, module=SetVariables_#5}
   &InitializeParticle(
    Color=&MakeCustomLinearColorFromBool(
     Bool=$Particles.bIsSplashes
     True Color=&RandomRangeLinearColor(Minimum=<rgba 4 4 4 1>, Maximum=<rgba 0.7835 0.7835 0.7835 0.6>)
     False Color=&RandomRangeLinearColor(Minimum=<rgba 3 3 3 1>, Maximum=<rgba 4 4 4 1>)
    )
    Lifetime=&MakeCustomFloatFromBool(
     Bool=$Particles.bIsSplashes
     True Float=&RandomRangeFloat(Minimum=0.85, Maximum=1.5)
     False Float=&RandomRangeFloat(Minimum=0.5, Maximum=0.75)
    )
    Mass Max=2
    Mass Mode=::NewEnumerator3
    Sprite Rotation Mode=::NewEnumerator2
    Sprite Size Mode=::NewEnumerator3
    Sprite UV Mode=::NewEnumerator3
    Uniform Sprite Size=&MakeCustomFloatFromBool(
     Bool=$Particles.bIsSplashes
     True Float=&RandomRangeFloat(Minimum=50, Maximum=75)
     False Float=&RandomRangeFloat(Minimum=7, Maximum=10)
    )
    UsePositionOffset=false
    Write Material Random=true
   )
   &ShapeLocation(
    MaskSpawnGroup=false
    Non Uniform Scale=&MakeCustomVectorFromBool(Bool=$Particles.bIsSplashes, True Vector=<0 0 0>, False Vector=<1 1 2>)
    Offset=&MakeCustomVectorFromBool(Bool=$Particles.bIsSplashes, True Vector=<0 0 0>, False Vector=<0 0 50>)
    Offset Mode=::NewEnumerator0
    Override Seed=false
    Sphere Radius=&MakeCustomFloatFromBool(Bool=$Particles.bIsSplashes, True Float=0, False Float=5)
    Sphere Surface Distribution=1
    UseOwnerScale=true
   )
   &AddVelocityFromPoint(
    Module.Use Falloff Distance=false
    Origin Offset=<0 0 52>
    Velocity Strength=&MakeCustomFloatFromBool(Bool=$Particles.bIsSplashes, True Float=0, False Float=&UniformRangedFloat(Minimum=500, Maximum=200))
   )
   &AddVelocity(Velocity=&MakeCustomVectorFromBool(Bool=$Particles.bIsSplashes, True Vector=<0 0 0>, False Vector=<0 0 400>))
   &StaticMeshLocation(
    Interpolate Mesh Sampling=false
    MaskSpawnGroup=false
    Mesh Sampling Type:ENiagara_StaticSamplingMode=::NewEnumerator1
    SamplePosOffset=true
    Sampled Position Offset:Vector3f=<0 0 -7>
    Sampled Position Offset Coordinate Space:ENiagaraSkelMeshTransforms=::NewEnumerator1
    Socket Index:NiagaraInt32=&ModuloInt(A=&SpawnGroup(), B=15)
    Socket Sampling Mode:ENiagara_StaticSocketMode=::NewEnumerator2
    Static Mesh:NiagaraDataInterfaceStaticMesh=$System.Static Mesh
   )
   &CameraOffset(Camera Offset Amount=&MakeCustomFloatFromBool(Bool=$Particles.bIsSplashes, True Float=15, False Float=0))
   &AddVelocity(
    Cone Angle=60
    Cone Axis=&ReturnMeshOrientationAxis(Mesh Orientation Axis=::XAxis, Mesh Orientation Quat=$Particles.StaticMeshLocation.SampledRotation)
    Module.UseConeFalloff=false
    Module.UseVelDistribution=false
    Override Seed=false
    Velocity Mode=::NewEnumerator2
    Velocity Speed=&MakeCustomFloatFromBool(Bool=$Particles.bIsSplashes, True Float=600, False Float=0)
   ) {module=AddVelocity003}
   &DynamicMaterialParameters(
    Index 0 Param 1=&RandomRangeFloat(Minimum=1, Maximum=2)
    Index 0 Param 2=0.5
    Index 0 Param 3=&RandomRangeFloat(Minimum=15, Maximum=50)
    Index 0 Param 4=&RandomRangeFloat(Minimum=2, Maximum=10)
    Index 1 Param 1=&RandomRangeFloat(Minimum=0, Maximum=10)
    Module.Param0WriteEnabled=true
    Module.Param0WriteEnabled1=true
    Module.Param1WriteEnabled=true
    Module.Param1WriteEnabled1=false
    Module.Param2WriteEnabled=true
    Module.Param2WriteEnabled1=false
    Module.Param3WriteEnabled=true
    Module.Param3WriteEnabled1=false
    Write Parameter Index 1=true
   )
   /SetVariables(Particles.VisibilityTag=&MakeCustomIntFromBool(Bool=$Particles.bIsSplashes, True INT=0, False INT=1)) {isSetParameters=true, module=SetVariables_#6}
  ]
  particleUpdateScript [
   &ParticleState()
   &SpriteRotationRate(Rotation Rate=&MakeCustomFloatFromBool(
     Bool=$Particles.bIsSplashes
     True Float=&Multiply_Float(
      A=&Lerp_Float(
       A=180
       B=45
       Alpha=&RemapRange(Input Value To Remap=$Particles.Mass, Input Min=0.75, Input Max=2, Output Min=0, Output Max=1)
      )
      B=&MakeFloatFromInt(Int=&RandomRangeInt(Minimum=-1, Maximum=1))
     )
     False Float=0
    ))
   &Drag(
    Drag=&MakeCustomFloatFromBool(
     Bool=$Particles.bIsSplashes
     True Float=&FloatFromCurve(
      FloatCurve=DataInterface() {propertyValues {Curve {keys [<at 0 2 RCIM_Cubic>,<at 0.5 0>]}}}
      Scale Curve=&RandomRangeFloat(Minimum=2, Maximum=2.5)
     )
     False Float=&FloatFromCurve(
      FloatCurve=DataInterface() {propertyValues {Curve {keys [<at 0 1 RCIM_Cubic>, {arriveTangent=-1.4297, interpMode=RCIM_Cubic, leaveTangent=-1.4297, tangentMode=RCTM_User, time=1, value=0}]}}}
      Scale Curve=&UniformRangedFloat(Minimum=10, Maximum=8)
     )
    )
    Use Linear Drag=true
    Use Rotational Drag=true
   )
   &GravityForce(Gravity=&MakeCustomVectorFromBool(Bool=$Particles.bIsSplashes, True Vector=<0 0 -980>, False Vector=<0 0 -1470>))
   &SolveForcesAndVelocity(Clamp Velocity=false, Limit Acceleration=false)
   &ScaleSpriteSizeBySpeed(
    Max Scale Factor=&MakeCustomVector2DFromBool(Bool=$Particles.bIsSplashes, True Vector2D=<1 1>, False Vector2D=<1.2 4>)
    Min Scale Factor=&MakeCustomVector2DFromBool(Bool=$Particles.bIsSplashes, True Vector2D=<1 1>, False Vector2D=<0.25 0.2>)
   )
   &ScaleSpriteSize(Uniform Curve Sprite Scale=DataInterface() {propertyValues {Curve {keys [
        {arriveTangent=1.8917, interpMode=RCIM_Cubic, leaveTangent=1.8917, tangentMode=RCTM_User, time=0, value=0.35}
        {interpMode=RCIM_Cubic, leaveTangent=-0.5754, time=1, value=2}
       ]}}})
   &ScaleColor(
    Scale Alpha=&MakeCustomFloatFromBool(
     Bool=$Particles.bIsSplashes
     True Float=&FloatFromCurve(FloatCurve=DataInterface() {propertyValues {Curve {keys [
          <at 0 -1>
          <at 0.0002 0.5 RCIM_Cubic>
          {arriveTangent=-2.4995, interpMode=RCIM_Cubic, leaveTangent=-2.4995, tangentMode=RCTM_User, time=1, value=-1}
         ]}}})
     False Float=&FloatFromCurve(FloatCurve=DataInterface() {propertyValues {Curve {keys [<at 0 1 RCIM_Cubic>, {arriveTangent=-3.0079, interpMode=RCIM_Cubic, leaveTangent=-3.0079, tangentMode=RCTM_User, time=1, value=0}]}}})
    )
    Scale RGB=&MakeCustomVectorFromBool(
     Bool=$Particles.bIsSplashes
     True Vector=&MakeVectorFromLinearColorRGB(Color=&MultiplyLinearColor(
       Linear Color=&ColorFromCurve(ColorCurve=DataInterface() {propertyValues {
          AlphaCurve {keys []}
          BlueCurve {keys [<at 0.1481 40>, {arriveTangent=-41.3655, interpMode=RCIM_Cubic, leaveTangent=-41.3655, time=0.198, value=12}, <at 0.9216 8>, <at 1 8>]}
          GreenCurve {keys [<at 0.1481 28.1816>, {arriveTangent=-26.0882, interpMode=RCIM_Cubic, leaveTangent=-26.0882, time=0.198, value=12}, <at 0.9216 8>, <at 1 8>]}
          RedCurve {keys [<at 0.1481 7.2711>, {arriveTangent=0.9422, interpMode=RCIM_Cubic, leaveTangent=0.9422, time=0.198, value=12}, <at 0.9216 8>, <at 1 8>]}
         }})
       Scale Factor=&MakeVector4_FromLinearColor(LinearColor=&LerpLinearColors(StartColor=<rgba 0.0019 0.5703 0.6979 1>, EndColor=<rgba 0.002 0.724 0.6397 1>, LerpFactor=$Particles.MaterialRandom))
      ))
     False Vector=<1 1 1>
    )
    ScaleA=true
    ScaleRGB=true
   )
  ]
  rendererClasses [&NiagaraSpriteRendererProperties]
  renderers [
   {rendererClass=&NiagaraSpriteRendererProperties}
   {rendererClass=&NiagaraSpriteRendererProperties, rendererIndex=1}
  ]
 }
 Fruit_Drops {
  bEnabled=false
  particleSpawnScript [
   &InitializeParticle(
    Color Maximum=<rgba 4 4 4 1>
    Color Minimum=<rgba 3 3 3 1>
    Color Mode=::NewEnumerator2
    Lifetime Max=0.75
    Lifetime Min=0.5
    Lifetime Mode=::NewEnumerator1
    Sprite Rotation Mode=::NewEnumerator2
    Sprite Size Mode=::NewEnumerator1
    Uniform Sprite Size Max=7
    Uniform Sprite Size Min=10
    UsePositionOffset=false
    Write Material Random=true
   ) {enabled=false}
   &ShapeLocation(
    MaskSpawnGroup=false
    Non Uniform Scale=<1 1 2>
    Offset=<0 0 50>
    Offset Mode=::NewEnumerator0
    Override Seed=false
    Sphere Radius=5
    Sphere Surface Distribution=1
    UseOwnerScale=true
   ) {enabled=false}
   &AddVelocityFromPoint(
    Module.Use Falloff Distance=false
    Origin Offset=<0 0 52>
    Velocity Strength=&UniformRangedFloat(Minimum=500, Maximum=200)
   ) {enabled=false}
   &AddVelocity(Velocity=<0 0 400>) {enabled=false}
   &DynamicMaterialParameters(
    Index 0 Param 1=&RandomRangeFloat(Minimum=1, Maximum=2)
    Index 0 Param 2=0.5
    Index 0 Param 3=&RandomRangeFloat(Minimum=15, Maximum=50)
    Index 0 Param 4=&RandomRangeFloat(Minimum=2, Maximum=10)
    Index 1 Param 1=&RandomRangeFloat(Minimum=0, Maximum=10)
    Module.Param0WriteEnabled=true
    Module.Param0WriteEnabled1=true
    Module.Param1WriteEnabled=true
    Module.Param1WriteEnabled1=false
    Module.Param2WriteEnabled=true
    Module.Param2WriteEnabled1=false
    Module.Param3WriteEnabled=true
    Module.Param3WriteEnabled1=false
    Write Parameter Index 1=true
   ) {enabled=false}
   &StaticMeshLocation(
    Interpolate Mesh Sampling=false
    MaskSpawnGroup=false
    Mesh Sampling Type:ENiagara_StaticSamplingMode=::NewEnumerator1
    SamplePosOffset=true
    Sampled Position Offset:Vector3f=<0 0 -7>
    Sampled Position Offset Coordinate Space:ENiagaraSkelMeshTransforms=::NewEnumerator1
    Socket Index:NiagaraInt32=&SpawnGroup()
    Socket Sampling Mode:ENiagara_StaticSocketMode=::NewEnumerator2
    Static Mesh:NiagaraDataInterfaceStaticMesh=$System.Static Mesh
   ) {enabled=false}
  ]
  particleUpdateScript [
   &ParticleState() {enabled=false}
   &GravityForce(Gravity=<0 0 -1470>) {enabled=false}
   &Drag(
    Drag=&FloatFromCurve(
     FloatCurve=DataInterface() {propertyValues {Curve {keys [<at 0 1 RCIM_Cubic>, {arriveTangent=-1.4297, interpMode=RCIM_Cubic, leaveTangent=-1.4297, tangentMode=RCTM_User, time=1, value=0}]}}}
     Scale Curve=&UniformRangedFloat(Minimum=10, Maximum=8)
    )
    Use Linear Drag=true
    Use Rotational Drag=true
   ) {enabled=false}
   &SolveForcesAndVelocity(Clamp Velocity=false, Limit Acceleration=false) {enabled=false}
   &ScaleSpriteSizeBySpeed(Max Scale Factor=<1.2 4>, Min Scale Factor=<0.25 0.2>) {enabled=false}
   &ScaleColor(
    Scale Alpha=&FloatFromCurve(FloatCurve=DataInterface() {propertyValues {Curve {keys [<at 0 1 RCIM_Cubic>, {arriveTangent=-3.0079, interpMode=RCIM_Cubic, leaveTangent=-3.0079, tangentMode=RCTM_User, time=1, value=0}]}}})
    ScaleA=true
    ScaleRGB=true
   ) {enabled=false}
  ]
  rendererClasses []
  renderers []
 }
 Gloop_Merged {
  emitterUpdateScript [
   &EmitterState()
   &SpawnBurst_Instantaneous(
    Spawn Count=3
    Spawn Time=&MakeCustomFloatFromBool(Bool=$User.WasHitByVeh, True Float=0, False Float=0.2)
    Use Spawn Probability=false
    UseLoopCountLimit=false
   )
   &SpawnBurst_Instantaneous(
    Spawn Count=&RandomRangeInt(Minimum=4, Maximum=12)
    Spawn Group=1
    Spawn Time=&MakeCustomFloatFromBool(Bool=$User.WasHitByVeh, True Float=0, False Float=0.2)
    Use Spawn Probability=false
    UseLoopCountLimit=false
   ) {module=SpawnBurst_Instantaneous001}
   &SpawnBurst_Instantaneous(
    Spawn Group=2
    Spawn Time=&MakeCustomFloatFromBool(Bool=$User.WasHitByVeh, True Float=0, False Float=0.2)
    Use Spawn Probability=false
    UseLoopCountLimit=false
   ) {module=SpawnBurst_Instantaneous002}
  ]
  particleSpawnScript [
   &InitializeParticle(
    Color=&MaskLinearColorBySpawnGroup(Masked Linear Color=<rgba 1 1 1 0.5>, Passthrough Linear Color=<rgba 1 1 1 1>, Spawn Group Mask=2)
    Lifetime=&self:Mask_FloatBy_3_Spawngroups(Spawngroup:NiagaraInt32=&SpawnGroup(), Zero:NiagaraFloat=0.62, One:NiagaraFloat=0.5, two:NiagaraFloat=0.6)
    Sprite Rotation Angle=&MaskFloatBySpawnGroup(
     Masked Float=&RandomRangeFloat(Minimum=-7, Maximum=7)
     Passthrough Float=&RandomRangeFloat(Minimum=360, Maximum=360)
     Spawn Group Mask=2
    )
    Sprite Rotation Mode=::NewEnumerator0
    Sprite Size Mode=::NewEnumerator3
    Uniform Sprite Size=&self:Mask_FloatBy_3_Spawngroups(
     Spawngroup:NiagaraInt32=&SpawnGroup()
     Zero:NiagaraFloat=&RandomRangeFloat(Minimum=120, Maximum=170)
     One:NiagaraFloat=&RandomRangeFloat(Minimum=100, Maximum=200)
     two:NiagaraFloat=&MakeFloatFromVector2d(Vector 2D=&Vector2DFromFloat(Value=&UniformRangedFloat(Minimum=350, Maximum=425)))
    )
    UsePositionOffset=false
    Write Material Random=true
   )
   /Niagara/Modules/Update/SubUV/SubUVAnimation(Number Of Frames=3, SubUV Animation Mode=::NewEnumerator1)
   &ShapeLocation(
    Hemisphere Distribution=&MaskVector2dBySpawnGroup(Masked Vector2D=<0 0.5>, Passthrough Vector2D=<1 1>, Spawn Group Mask=2)
    MaskSpawnGroup=false
    Offset=&MaskVectorBySpawnGroup(Masked Vector=<0 0 15>, Passthrough Vector=<0 0 50>, Spawn Group Mask=2)
    Offset Mode=::NewEnumerator0
    Override Seed=false
    Sphere Radius=&self:Mask_FloatBy_3_Spawngroups(Spawngroup:NiagaraInt32=&SpawnGroup(), Zero:NiagaraFloat=0, One:NiagaraFloat=15, two:NiagaraFloat=5)
    UseOwnerScale=true
   ) {module=ShapeLocation001}
   &AddVelocity(Velocity=&self:ScratchDynamicInput(
     SpawnGroup:NiagaraInt32=&SpawnGroup()
     Zero:Vector3f=&UniformRangedVector(Minimum=<0 0 300>, Maximum=<1 1 225>)
     One:Vector3f=<0 0 1200>
     Two:Vector3f=<0 0 500>
    )) {module=AddVelocity002}
   &AddVelocity(
    Module.Use Falloff Distance=false
    Origin Offset=&MaskVectorBySpawnGroup(Masked Vector=<0 0 0>, Passthrough Vector=<0 0 50>)
    Override Seed=false
    Velocity Mode=::NewEnumerator1
    Velocity Speed=&self:Mask_FloatBy_3_Spawngroups(
     Spawngroup:NiagaraInt32=&SpawnGroup()
     Zero:NiagaraFloat=0
     One:NiagaraFloat=&RandomRangeFloat(Minimum=700, Maximum=1200)
     two:NiagaraFloat=0
    )
   )
   /SetVariables() {isSetParameters=true}
  ]
  particleUpdateScript [
   &ParticleState()
   &ScaleSpriteSize(
    Scale Sprite Size Mode=::NewEnumerator0
    Uniform Scale Factor=&self:Mask_FloatBy_3_Spawngroups(
     Spawngroup:NiagaraInt32=&SpawnGroup()
     Zero:NiagaraFloat=1
     One:NiagaraFloat=1
     two:NiagaraFloat=&FloatFromCurve(
      FloatCurve=DataInterface() {propertyValues {Curve {keys [
          {interpMode=RCIM_Cubic, leaveTangent=0.0012, tangentMode=RCTM_Break, time=0, value=0}
          {arriveTangent=0.0008, interpMode=RCIM_Cubic, tangentMode=RCTM_Break, time=700, value=1}
         ]}}}
      CurveIndex=$Engine.Owner.LODDistance
     )
    )
   ) {module=ScaleSpriteSize001}
   &ScaleSpriteSize(
    Scale Factor=&self:ScratchDynamicInput_02(
     SpawnGroup:NiagaraInt32=&SpawnGroup()
     Zero:Vector2f=<1 0.8>
     One:Vector2f=<0.7 1>
     Two:Vector2f=&MakeVector2D(X=&RandomRangeFloat(Minimum=1.1, Maximum=1), Y=1.1)
    )
    Scale Sprite Size Mode=::NewEnumerator2
   ) {module=ScaleSpriteSize002}
   &AccelerationForce(Acceleration=&self:ScratchDynamicInput(
     SpawnGroup:NiagaraInt32=&SpawnGroup()
     Zero:Vector3f=&UniformRangedVector(Minimum=<0 0 -800>, Maximum=<0 0 -600>)
     One:Vector3f=&UniformRangedVector(Minimum=<0 0 -800>, Maximum=<0 0 -1000>)
     Two:Vector3f=<0 0 -600>
    )) {module=AccelerationForce001}
   &Drag(
    Drag=&self:Mask_FloatBy_3_Spawngroups(Spawngroup:NiagaraInt32=&SpawnGroup(), Zero:NiagaraFloat=0, One:NiagaraFloat=6, two:NiagaraFloat=2)
    Use Linear Drag=true
    Use Rotational Drag=false
   )
   &SolveForcesAndVelocity(Clamp Velocity=false, Limit Acceleration=false)
   &ScaleColor(
    Scale RGB=&MaskVectorBySpawnGroup(
     Masked Vector=<1 1 1>
     Passthrough Vector=&VectorFromCurve(VectorCurve=DataInterface() {propertyValues {XCurve {keys [<at 0 1.5>,<at 1 1>]}, YCurve {keys [<at 0 1.5>,<at 1 1>]}, ZCurve {keys [<at 0 1.5>,<at 1 1>]}}})
     Spawn Group Mask=2
    )
    ScaleA=true
    ScaleRGB=true
   )
   &ScaleSpriteSize(Scale Factor=&Vector2DFromFloat(Value=&MaskFloatBySpawnGroup(
      Masked Float=&FloatFromCurve(FloatCurve=DataInterface() {propertyValues {Curve {keys [
           {arriveTangent=7.4318, interpMode=RCIM_Cubic, leaveTangent=7.4318, tangentMode=RCTM_User, time=0, value=0}
           {arriveTangent=3.334, interpMode=RCIM_Cubic, leaveTangent=3.334, tangentMode=RCTM_User, time=0.1723, value=1.192}
           <at 2 3 RCIM_Cubic>
          ]}}})
      Passthrough Float=1
     )))
   &DynamicMaterialParameters(
    Index 0 Param 1=&FloatFromCurve(FloatCurve=DataInterface() {propertyValues {Curve {keys [
         {interpMode=RCIM_Cubic, tangentMode=RCTM_Break, time=0, value=1}
         {arriveTangent=-1.1592, interpMode=RCIM_Cubic, leaveTangent=-1.1592, tangentMode=RCTM_User, time=1, value=0}
        ]}}})
    Index 0 Param 2=&FloatFromCurve(FloatCurve=DataInterface() {propertyValues {Curve {keys [{arriveTangent=0.4361, interpMode=RCIM_Cubic, leaveTangent=0.4361, tangentMode=RCTM_User, time=0, value=0.05}, <at 1 0.25>]}}})
    Module.Param0WriteEnabled=true
    Module.Param1WriteEnabled=true
    Module.Param2WriteEnabled=true
    Module.Param3WriteEnabled=true
   )
   &CameraOffset(Camera Offset Amount=&FloatFromCurve(FloatCurve=DataInterface() {propertyValues {Curve {keys [
         {interpMode=RCIM_Cubic, leaveTangent=1080.3575, tangentMode=RCTM_Break, time=0, value=15}
         {arriveTangent=135, interpMode=RCIM_Cubic, leaveTangent=135, tangentMode=RCTM_Break, time=0.12, value=94}
         <at 1 150>
        ]}}}))
   &SpriteRotationRate(Rotation Rate=&MaskFloatBySpawnGroup(
     Masked Float=&Multiply_Float(
      A=&Lerp_Float(A=-1, B=1, Alpha=$Particles.Initial.MaterialRandom)
      B=&Lerp_Float(A=125, B=0, Alpha=&Power(a=$Particles.NormalizedAge, b=0.2))
     )
     Passthrough Float=1
    ))
   /SetVariables(Particles.VisibilityTag=&SpawnGroup()) {isSetParameters=true}
   /SetVariables(Particles.Combined Dynamic0:Vector4f=&MakeVector4(
     X=&FloatFromCurve(FloatCurve=DataInterface() {propertyValues {Curve {keys [<at 0 0.01>, {arriveTangent=0.6025, interpMode=RCIM_Cubic, leaveTangent=-0.6448, tangentMode=RCTM_Break, time=1, value=1}]}}})
     Y=&FloatFromCurve(FloatCurve=DataInterface() {propertyValues {Curve {keys [{arriveTangent=0.4361, time=0.2, value=0.05}, <at 1 0.7>]}}})
     Z=&FloatFromCurve(FloatCurve=DataInterface() {propertyValues {Curve {keys [
          {arriveTangent=0.4361, interpMode=RCIM_Cubic, tangentMode=RCTM_Break, time=0.5, value=0}
          {arriveTangent=-1.6488, interpMode=RCIM_Cubic, tangentMode=RCTM_Break, time=1, value=-0.5}
         ]}}})
     W=1
    )) {isSetParameters=true}
   &Color() {enabled=false}
  ]
  rendererClasses [&NiagaraSpriteRendererProperties]
  renderers [
   {rendererClass=&NiagaraSpriteRendererProperties}
   {rendererClass=&NiagaraSpriteRendererProperties, rendererIndex=1}
   {rendererClass=&NiagaraSpriteRendererProperties, rendererIndex=2}
  ]
 }
 SlurpWispySmoke {
  emitterUpdateScript [
   &EmitterState(
    Enable Distance Culling=true
    MaxDistance=20000
    Module.bUseMaxDistance=true
    Module.bUseMinDistance=false
    Scalability Mode=::NewEnumerator1
   )
   &SpawnBurst_Instantaneous(
    Spawn Count=4
    Spawn Time=&MakeCustomFloatFromBool(Bool=$User.WasHitByVeh, True Float=0, False Float=0.2)
    Use Spawn Probability=false
    UseLoopCountLimit=false
   )
  ]
  particleSpawnScript [
   &InitializeParticle(
    Color Maximum=<rgba 1 1 1 0.425>
    Color Minimum=<rgba 0.25 0.25 0.25 0.35>
    Color Mode=::NewEnumerator2
    Lifetime Min=1.8
    Lifetime Mode=::NewEnumerator1
    Mass Mode=::NewEnumerator3
    Sprite Rotation Mode=::NewEnumerator2
    Sprite Size Mode=::NewEnumerator1
    Sprite UV Mode=::NewEnumerator3
    Uniform Sprite Size Max=350
    Uniform Sprite Size Min=275
    UsePositionOffset=false
    Write Material Random=true
   )
   &ShapeLocation(
    MaskSpawnGroup=false
    Sphere Distribution=::NewEnumerator1
    Sphere Radius=10
    U Position=&ReturnNormalizedExecIndex()
    UseOwnerScale=true
   )
   &AddVelocity(
    Velocity=&AddVectorToPosition(Position=$Particles.ShapeLocation.ShapeNormal, Vector=<0 0 1>)
    Velocity Speed Scale=450
   )
   &AddVelocity(Module.Use Falloff Distance=false, Origin Offset=<0 0 -20>, Override Seed=false, Velocity Mode=::NewEnumerator1, Velocity Speed=800) {module=AddVelocity001}
   &Color(Color=&MultiplyLinearColor(
     Linear Color=$Particles.Color
     Scale Factor=&MakeVector4_FromLinearColor(LinearColor=&UniformRangedLinearColor(Minimum=<rgba 0 1 0.8913 0.23>, Maximum=<rgba 0 0.4471 1 0.23>))
    ))
   &CameraOffset(Camera Offset Amount=&RandomRangeFloat(Minimum=40, Maximum=50))
  ]
  particleUpdateScript [
   &ParticleState()
   &Drag(Drag=&UniformRangedFloat(Minimum=3, Maximum=4), Use Linear Drag=true, Use Rotational Drag=true)
   &GravityForce()
   &SolveForcesAndVelocity(Clamp Velocity=false, Limit Acceleration=false)
   &SpriteRotationRate(Rotation Rate=&Multiply_Float(
     A=&Lerp_Float(A=90, B=-90, Alpha=$Particles.Initial.MaterialRandom)
     B=&Lerp_Float(A=1, B=0, Alpha=$Particles.NormalizedAge)
    ))
   &ScaleColor(
    Scale Alpha=&FloatFromCurve(
     FloatCurve=DataInterface() {propertyValues {Curve {keys [{interpMode=RCIM_Cubic, tangentMode=RCTM_User, time=0, value=1}, {interpMode=RCIM_Cubic, tangentMode=RCTM_User, time=1, value=0}]}}}
     CurveIndex=&Power(a=$Particles.NormalizedAge, b=0.5)
    )
    ScaleA=true
    ScaleRGB=true
   )
   &ScaleSpriteSize(Uniform Curve Sprite Scale=DataInterface() {propertyValues {Curve {keys [{arriveTangent=4.1807, interpMode=RCIM_Cubic, leaveTangent=4.1807, tangentMode=RCTM_User, time=0, value=0.75}, <at 1 2 RCIM_Cubic>]}}})
   &SubUVAnimation(Number Of Frames=63)
  ]
  rendererClasses [&NiagaraSpriteRendererProperties]
  renderers [{rendererClass=&NiagaraSpriteRendererProperties}]
 }
 Soda_pop {
  simTarget=GPUComputeSim
  emitterUpdateScript [
   &EmitterState(
    Enable Distance Culling=true
    Life Cycle Mode=::NewEnumerator1
    Loop Behavior=::NewEnumerator1
    Loop Delay=&MakeCustomFloatFromBool(Bool=$User.WasHitByVeh, True Float=0, False Float=0.2)
    Loop Duration=1
    MaxDistance=3000
    Module.bUseMaxDistance=true
    Module.bUseMinDistance=false
    Scalability Mode=::NewEnumerator1
    UseLoopDelay=true
   )
   &SpawnRate(SpawnRate=150, Use Spawn Probability=false)
  ]
  particleSpawnScript [
   &InitializeParticle(
    Color Maximum=<rgba 0 1.79 4 1>
    Color Minimum=<rgba 0 4 1.2666 1>
    Color Mode=::NewEnumerator2
    Lifetime Max=1
    Lifetime Min=0.2
    Lifetime Mode=::NewEnumerator1
    Sprite Rotation Mode=::NewEnumerator2
    Sprite Size Mode=::NewEnumerator1
    Uniform Sprite Size Max=12
    Uniform Sprite Size Min=4
    UsePositionOffset=false
    Write Material Random=true
   )
   &ShapeLocation(
    Cylinder Height=10
    Cylinder Height Midpoint=0.1
    Cylinder Radius=300
    MaskSpawnGroup=false
    Override Seed=false
    Shape Primitive=::NewEnumerator1
    Surface Only=false
    UseOwnerScale=true
   )
   &AddVelocity(Velocity=<0 0 150>)
  ]
  particleUpdateScript [
   &ParticleState()
   /Niagara/Modules/Update/Forces/CurlNoiseForce(Noise Strength=800, Pan Noise=false)
   &SolveForcesAndVelocity(Clamp Velocity=false, Limit Acceleration=false)
   &ScaleSpriteSize(Scale Factor=&Vector2DFromFloat(Value=&FloatFromCurve(FloatCurve=DataInterface() {propertyValues {Curve {keys [{arriveTangent=3.6722, interpMode=RCIM_Cubic, leaveTangent=3.6722, tangentMode=RCTM_User, time=0, value=0}, <at 0.8 1>, <at 1 2>]}}})))
   &ScaleColor(
    Scale Alpha=&FloatFromCurve(FloatCurve=DataInterface() {propertyValues {Curve {keys [
         <at 0 1>
         {arriveTangent=-4.4715, interpMode=RCIM_Cubic, leaveTangent=-4.4715, tangentMode=RCTM_User, time=0.8022, value=0.9932}
         {arriveTangent=-7.8513, interpMode=RCIM_Cubic, leaveTangent=-7.8513, tangentMode=RCTM_User, time=1, value=0}
        ]}}})
    ScaleA=true
    ScaleRGB=true
   )
  ]
  rendererClasses [&NiagaraSpriteRendererProperties]
  renderers [{rendererClass=&NiagaraSpriteRendererProperties}]
 }
 Splash001 {
  bEnabled=false
  emitterUpdateScript [
   &EmitterState() {enabled=false}
   &SpawnBurst_Instantaneous(
    Spawn Count=1
    Spawn Time=&MakeCustomFloatFromBool(Bool=$User.WasHitByVeh, True Float=0, False Float=0.2)
    Use Spawn Probability=false
   ) {enabled=false}
  ]
  particleSpawnScript [
   &InitializeParticle(
    Color=<rgba 1 1 1 0.5>
    Lifetime=0.6
    Mass Mode=::NewEnumerator1
    Sprite Rotation Angle=&UniformRangedFloat(Minimum=-7.5, Maximum=7.5)
    Sprite Rotation Mode=::NewEnumerator0
    Sprite Size=&Vector2DFromFloat(Value=&UniformRangedFloat(Minimum=350, Maximum=425))
    Sprite Size Mode=::NewEnumerator4
    UsePositionOffset=false
    Write Material Random=true
   ) {enabled=false, module=InitializeParticle001}
   &SystemLocation(MaskSpawnGroup=false, Offset=<0 0 15>) {enabled=false}
   &SphereLocation(
    Hemisphere Z=true
    MaskSpawnGroup=false
    Override Seed=false
    Sphere Radius=5
    Surface Only=false
    UseNonUniformScale=false
    UseOffset=false
    UseSphereOrientation=false
   ) {enabled=false}
   &AddVelocity(Velocity=<0 0 500>) {enabled=false}
  ]
  particleUpdateScript [
   &ParticleState() {enabled=false}
   &ScaleSpriteSize(
    Scale Factor=&MakeVector2D(
     X=&Multiply_Float(
      A=&MakeFloatFromInt(Int=/Niagara/DynamicInputs/Subtract/Subtract_Integer(
        A:NiagaraInt32=&Multiply_Int(A=&RandomRangeInt(Minimum=0, Maximum=1), B=2)
        B:NiagaraInt32=1
       ))
      B=1.1
     )
     Y=1.1
    )
    Scale Sprite Size Mode=::NewEnumerator2
   ) {enabled=false}
   &AccelerationForce(Acceleration=<0 0 -600>) {enabled=false}
   &Drag(Drag=2, Use Linear Drag=true, Use Rotational Drag=false) {enabled=false}
   &SolveForcesAndVelocity(Clamp Velocity=false, Limit Acceleration=false) {enabled=false}
   &DynamicMaterialParameters(
    Index 0 Param 1=&FloatFromCurve(FloatCurve=DataInterface() {propertyValues {Curve {keys [<at 0 0.01>, {arriveTangent=0.6025, interpMode=RCIM_Cubic, leaveTangent=-0.6448, tangentMode=RCTM_Break, time=1, value=1}]}}})
    Index 0 Param 2=&FloatFromCurve(FloatCurve=DataInterface() {propertyValues {Curve {keys [{arriveTangent=0.4361, time=0.2, value=0.05}, <at 1 0.7>]}}})
    Index 0 Param 3=&FloatFromCurve(FloatCurve=DataInterface() {propertyValues {Curve {keys [
         {arriveTangent=0.4361, interpMode=RCIM_Cubic, tangentMode=RCTM_Break, time=0.5, value=0}
         {arriveTangent=-1.6488, interpMode=RCIM_Cubic, tangentMode=RCTM_Break, time=1, value=-0.5}
        ]}}})
    Module.Param0WriteEnabled=true
    Module.Param1WriteEnabled=true
    Module.Param2WriteEnabled=true
    Module.Param3WriteEnabled=true
   ) {enabled=false}
   &CameraOffset(Camera Offset Amount=80) {enabled=false}
   &CameraOffset(Camera Offset Amount=&FloatFromCurve(FloatCurve=DataInterface() {propertyValues {Curve {keys [
         {interpMode=RCIM_Cubic, leaveTangent=1080.3575, tangentMode=RCTM_Break, time=0, value=15}
         {arriveTangent=135, interpMode=RCIM_Cubic, leaveTangent=135, tangentMode=RCTM_Break, time=0.12, value=94}
         <at 1 150>
        ]}}})) {enabled=false, module=CameraOffset001}
   &ScaleSpriteSize(
    Scale Sprite Size Mode=::NewEnumerator0
    Uniform Scale Factor=&FloatFromCurve(
     FloatCurve=DataInterface() {propertyValues {Curve {keys [
         {interpMode=RCIM_Cubic, leaveTangent=0.0012, tangentMode=RCTM_Break, time=0, value=0}
         {arriveTangent=0.0008, interpMode=RCIM_Cubic, tangentMode=RCTM_Break, time=700, value=1}
        ]}}}
     CurveIndex=$Engine.Owner.LODDistance
    )
   ) {module=ScaleSpriteSize001}
  ]
  rendererClasses [&NiagaraSpriteRendererProperties]
  renderers [{rendererClass=&NiagaraSpriteRendererProperties}]
 }
 ThrownDrops002 {
  bEnabled=false
  simTarget=GPUComputeSim
  emitterUpdateScript [
   &EmitterState(
    Enable Distance Culling=true
    Life Cycle Mode=::NewEnumerator1
    Loop Behavior=::NewEnumerator1
    Loop Duration=1
    Module.bUseMaxDistance=true
    Module.bUseMinDistance=false
    Scalability Mode=::NewEnumerator1
    UseLoopDelay=false
   ) {enabled=false}
   &SpawnBurst_Instantaneous(
    Spawn Count=&UniformRangedInt(Minimum=22, Maximum=75)
    Spawn Time=&MakeCustomFloatFromBool(Bool=$User.WasHitByVeh, True Float=0, False Float=0.2)
    Use Spawn Probability=false
   ) {enabled=false}
  ]
  particleSpawnScript [
   &InitializeParticle(
    Color=&UniformRangedLinearColor(Minimum=<rgba 1 1 1 1>, Maximum=<rgba 1 1 0.2 1>)
    Lifetime=&UniformRangedFloat(Minimum=0.5, Maximum=0.75)
    Mass=&UniformRangedFloat(Minimum=0.75, Maximum=1.75)
    Mass Mode=::NewEnumerator1
    Sprite Size=&Vector2DFromFloat(Value=&UniformRangedFloat(Minimum=10, Maximum=20))
    Sprite Size Mode=::NewEnumerator4
    UsePositionOffset=false
    Write Material Random=true
   ) {enabled=false, module=InitializeParticle001}
   &SphereLocation(
    MaskSpawnGroup=false
    Offset=<0 0 50>
    Override Seed=false
    Sphere Radius=20
    Surface Only=false
    UseNonUniformScale=false
    UseOffset=true
    UseSphereOrientation=false
   ) {enabled=false}
   &AddVelocityFromPoint(
    Module.Use Falloff Distance=false
    Origin Offset=<0 0 50>
    Velocity Strength=&UniformRangedFloat(Minimum=1000, Maximum=2000)
   ) {enabled=false}
   &AddVelocity(Velocity=<0 0 800>) {enabled=false}
  ]
  particleUpdateScript [
   &ParticleState() {enabled=false}
   &GravityForce(Gravity=<0 0 -2250>)
   &Drag(Drag=&UniformRangedFloat(Minimum=1, Maximum=3), Use Linear Drag=true, Use Rotational Drag=false) {enabled=false}
   &ScaleColor(Scale Alpha=&Lerp_Float(A=1, B=0, Alpha=$Particles.NormalizedAge), ScaleA=true, ScaleRGB=true) {enabled=false}
   &SolveForcesAndVelocity(Clamp Velocity=false, Limit Acceleration=false) {enabled=false}
   &ScaleSpriteSizeBySpeed(Max Scale Factor=<1 5>, Min Scale Factor=<0 0.05>) {enabled=false}
  ]
  rendererClasses [&NiagaraSpriteRendererProperties]
  renderers [{rendererClass=&NiagaraSpriteRendererProperties}]
 }
 ThrownDropsMerged {
  emitterUpdateScript [
   &EmitterState(
    Enable Distance Culling=true
    Life Cycle Mode=::NewEnumerator1
    Loop Behavior=::NewEnumerator1
    Loop Duration=1
    MaxDistance=5000000
    Module.bUseMaxDistance=true
    Module.bUseMinDistance=false
    Scalability Mode=::NewEnumerator1
    UseLoopDelay=false
   )
   &SpawnBurst_Instantaneous(
    Spawn Count=&UniformRangedInt(Minimum=5, Maximum=20)
    Spawn Time=&MakeCustomFloatFromBool(Bool=$User.WasHitByVeh, True Float=0, False Float=0.2)
    Use Spawn Probability=false
   )
   &SpawnBurst_Instantaneous(
    Spawn Count=&UniformRangedInt(Minimum=20, Maximum=50)
    Spawn Group=1
    Spawn Time=&MakeCustomFloatFromBool(Bool=$User.WasHitByVeh, True Float=0, False Float=0.22)
    Use Spawn Probability=false
    UseLoopCountLimit=false
   ) {module=SpawnBurst_Instantaneous001}
  ]
  particleSpawnScript [
   &InitializeParticle(
    Color=&MaskLinearColorBySpawnGroup(
     Masked Linear Color=&UniformRangedLinearColor(Minimum=<rgba 1 1 1 1>, Maximum=<rgba 1 1 1 1>)
     Passthrough Linear Color=&UniformRangedLinearColor(Minimum=<rgba 1 1 1 1>, Maximum=<rgba 1 1 0.2 1>)
    )
    Lifetime=&MaskFloatBySpawnGroup(
     Masked Float=&UniformRangedFloat(Minimum=0.25, Maximum=1.5)
     Passthrough Float=&UniformRangedFloat(Minimum=0.5, Maximum=0.75)
    )
    Mass=&UniformRangedFloat(Minimum=0.75, Maximum=1.75)
    Mass Mode=::NewEnumerator1
    Sprite Size=&Vector2DFromFloat(Value=&UniformRangedFloat(Minimum=0.5, Maximum=4))
    Sprite Size Mode=::NewEnumerator4
    UsePositionOffset=false
    Write Material Random=true
   ) {module=InitializeParticle001}
   /Niagara/Modules/Update/Mass/CalculateSizeAndRotationalInertiaByMass(Depth=0, Module.Manually Enter Density=false)
   &SphereLocation(
    MaskSpawnGroup=false
    Offset=<0 0 50>
    Override Seed=false
    Sphere Radius=20
    Surface Only=false
    UseNonUniformScale=false
    UseOffset=true
    UseSphereOrientation=false
   )
   &AddVelocityFromPoint(
    Module.Use Falloff Distance=false
    Origin Offset=<0 0 50>
    Velocity Strength=&MaskFloatBySpawnGroup(
     Masked Float=&UniformRangedFloat(Minimum=150, Maximum=650)
     Passthrough Float=&UniformRangedFloat(Minimum=1000, Maximum=2000)
    )
   )
   &AddVelocity(Velocity=&MaskVectorBySpawnGroup(Masked Vector=<0 0 500>, Passthrough Vector=<0 0 800>))
  ]
  particleUpdateScript [
   &ParticleState()
   &GravityForce(Gravity=&MaskVectorBySpawnGroup(Masked Vector=<0 0 -1650>, Passthrough Vector=<0 0 -2250>))
   &Drag(
    Drag=&MaskFloatBySpawnGroup(
     Masked Float=&UniformRangedFloat(Minimum=2, Maximum=3)
     Passthrough Float=&UniformRangedFloat(Minimum=1, Maximum=3)
    )
    Use Linear Drag=true
    Use Rotational Drag=false
   )
   &ScaleColor(Scale Alpha=&Lerp_Float(A=1, B=0, Alpha=$Particles.NormalizedAge), ScaleA=true, ScaleRGB=true)
   &SolveForcesAndVelocity(Clamp Velocity=false, Limit Acceleration=false)
   &ScaleSpriteSizeBySpeed(
    Max Scale Factor=&MaskVector2dBySpawnGroup(Masked Vector2D=<2 8>, Passthrough Vector2D=<1 5>, Spawn Group Mask=0)
    Min Scale Factor=&MaskVector2dBySpawnGroup(Masked Vector2D=<1 1>, Passthrough Vector2D=<0 0.05>, Spawn Group Mask=0)
   )
  ]
  rendererClasses [&NiagaraSpriteRendererProperties]
  renderers [{rendererClass=&NiagaraSpriteRendererProperties}]
 }
 initSphere001 {
  bEnabled=false
  emitterUpdateScript [
   &EmitterState(
    Enable Distance Culling=true
    Life Cycle Mode=::NewEnumerator1
    Loop Behavior=::NewEnumerator1
    Loop Duration=0.1
    MaxDistance=13000
    Module.bUseMaxDistance=true
    Module.bUseMinDistance=false
    Scalability Mode=::NewEnumerator1
    UseLoopDelay=false
   )
   &SpawnBurst_Instantaneous(
    Spawn Count=1
    Spawn Group=1
    Spawn Time=&MakeCustomFloatFromBool(Bool=$User.WasHitByVeh, True Float=0, False Float=0.2)
    Use Spawn Probability=false
   ) {module=SpawnBurst_Instantaneous001}
   &SpawnBurst_Instantaneous(
    Spawn Count=1
    Spawn Time=&MakeCustomFloatFromBool(Bool=$User.WasHitByVeh, True Float=0, False Float=0.2)
    Use Spawn Probability=false
   )
  ]
  particleSpawnScript [
   /Niagara/Modules/Spawn/Initialization/InitializeParticle(
    Color=<rgba 0.4 0.4 0.4 1>
    Lifetime=0.7
    Mesh Scale=&VectorFromFloat(Value=1)
    Module.Write Color=true
    Module.Write Lifetime=true
    Module.Write Mass=true
    Module.Write Material Random=true
    Module.Write Position=true
    Module.Write Scale=true
    Module.Write SpriteRotation=false
    Module.Write SpriteSize=false
    Sprite Size=<5 5>
   ) {enabled=false}
   /Niagara/Modules/Spawn/Initialization/V2/InitializeParticle(
    Color=&MaskLinearColorBySpawnGroup(Masked Linear Color=<rgba 0.4 0.4 0.4 1>, Passthrough Linear Color=<rgba 0 0.8333 1 1>)
    Lifetime=&MaskFloatBySpawnGroup(Masked Float=0.7, Passthrough Float=0.25)
    Mesh Scale Mode=::NewEnumerator3
    Sprite Rotation Angle=90
    Sprite Rotation Mode=::NewEnumerator0
    Sprite Size=<1200 400>
    Sprite Size Mode=::NewEnumerator4
    UsePositionOffset=false
    Write Material Random=true
   ) {module=InitializeParticle001}
   &DynamicMaterialParameters(
    Index 0 Param 1=&RandomRangeFloat(Minimum=0.4, Maximum=0.6)
    Index 0 Param 2=1
    Index 0 Param 3=1
    Index 0 Param 4=1
    Module.Param0WriteEnabled=true
    Module.Param1WriteEnabled=true
    Module.Param2WriteEnabled=true
    Module.Param3WriteEnabled=true
   )
   &SystemLocation(MaskSpawnGroup=false, Offset=&MaskVectorBySpawnGroup(Masked Vector=<0 0 50>, Passthrough Vector=<0 0 100>))
   &CameraOffset(Camera Offset Amount=&MaskFloatBySpawnGroup(Masked Float=0, Passthrough Float=50))
  ]
  particleUpdateScript [
   &ParticleState()
   &ScaleSpriteSize(Scale Factor=&Vector2DFromFloat(Value=&FloatFromCurve(
      FloatCurve=DataInterface() {propertyValues {Curve {keys [<at 0 1.5 RCIM_Cubic>,<at 1 0>]}}}
      Scale Curve=&FN_FrontEndFloatModifier(Gameplay=1, Front End=0.6)
     )))
   &ScaleMeshSize(Scale Factor=&VectorFromFloat(Value=&FloatFromCurve(
      FloatCurve=DataInterface() {propertyValues {Curve {keys [
          {interpMode=RCIM_Cubic, leaveTangent=2.1274, tangentMode=RCTM_Break, time=0, value=0}
          {arriveTangent=0.6694, interpMode=RCIM_Cubic, leaveTangent=0.6694, tangentMode=RCTM_User, time=0.494, value=0.654}
          <at 1.5 1>
         ]}}}
      Scale Curve=11
     )))
   &ScaleColor(
    Scale Alpha=&MaskFloatBySpawnGroup(
     Masked Float=&FloatFromCurve(FloatCurve=DataInterface() {propertyValues {Curve {keys [<at 0 0>,<at 0.2 1>,<at 0.8 1>,<at 1 0>]}}})
     Passthrough Float=1
    )
    Scale RGB=&MaskVectorBySpawnGroup(
     Masked Vector=&MakeVectorFromLinearColorRGB(Color=&ColorFromCurve(ColorCurve=DataInterface() {propertyValues {
         AlphaCurve {keys [{arriveTangent=0.9905, interpMode=RCIM_Cubic, time=0, value=1}, <at 1 1 RCIM_Cubic>]}
         BlueCurve {keys [<at 0 1>, {arriveTangent=-1, interpMode=RCIM_Cubic, leaveTangent=-1, time=0.2192, value=1}, <at 1 0>]}
         GreenCurve {keys [<at 0 0.6667>, {arriveTangent=-0.6667, interpMode=RCIM_Cubic, leaveTangent=-0.6667, time=0.2192, value=0.5}, <at 1 0>]}
         RedCurve {keys [<at 0 0>,<at 0.2192 0 RCIM_Cubic>,<at 1 0>]}
        }}))
     Passthrough Vector=<2 2 2>
    )
    ScaleA=true
    ScaleRGB=true
   )
   /SetVariables(Particles.LightColor=&ColorFromCurve(ColorCurve=DataInterface() {propertyValues {
       AlphaCurve {keys [<at 0.5925 1 RCIM_Cubic>,<at 1 0 RCIM_Cubic>]}
       BlueCurve {keys [<at -0.0022 0>, {arriveTangent=1.6758, interpMode=RCIM_Cubic, leaveTangent=1.6758, time=0.0394, value=1}, <at 0.5946 1>, <at 1 0 RCIM_Cubic>]}
       GreenCurve {keys [
         <at -0.0022 0>
         {arriveTangent=0.2015, interpMode=RCIM_Cubic, leaveTangent=0.2015, time=0.0394, value=0.7569}
         <at 0.5946 0.1203>
         <at 1 0 RCIM_Cubic>
        ]}
       RedCurve {keys [<at -0.0022 0>,<at 0.0394 0.2292 RCIM_Cubic>,<at 0.5946 0>,<at 1 0 RCIM_Cubic>]}
      }})) {isSetParameters=true}
   /SetVariables(Particles.VisibilityTag=&SpawnGroup()) {isSetParameters=true}
  ]
  rendererClasses [
   &NiagaraMeshRendererProperties
   &NiagaraComponentRendererProperties
   &NiagaraSpriteRendererProperties
  ]
  renderers [
   {rendererClass=&NiagaraMeshRendererProperties, rendererIndex=0}
   {rendererClass=&NiagaraComponentRendererProperties, rendererIndex=1}
   {rendererClass=&NiagaraSpriteRendererProperties, rendererIndex=2}
  ]
 }
}
summary {
 systemName=NS_GravBerryCactus_SlurpExplosion
 emitters [
  {emitterName=Soda_pop, simTarget=GPUComputeSim, rendererClasses [&NiagaraSpriteRendererProperties]}
  {emitterName=SlurpWispySmoke, rendererClasses [&NiagaraSpriteRendererProperties]}
  {bEnabled=false, emitterName=Fruit_Drops, rendererClasses []}
  {emitterName=Decal, rendererClasses [&NiagaraDecalRendererProperties]}
  {emitterName=ThrownDropsMerged, rendererClasses [&NiagaraSpriteRendererProperties]}
  {bEnabled=false, emitterName=ThrownDrops002, simTarget=GPUComputeSim, rendererClasses [&NiagaraSpriteRendererProperties]}
  {emitterName=FN_PIC_FLIP_Fluid_Explosion001, simTarget=GPUComputeSim, rendererClasses [&NiagaraSpriteRendererProperties]}
  {bEnabled=false, emitterName=Splash001, rendererClasses [&NiagaraSpriteRendererProperties]}
  {emitterName=Gloop_Merged, rendererClasses [&NiagaraSpriteRendererProperties]}
  {
   bEnabled=false
   emitterName=initSphere001
   rendererClasses [
    &NiagaraMeshRendererProperties
    &NiagaraComponentRendererProperties
    &NiagaraSpriteRendererProperties
   ]
  }
  {emitterName=FruitGlowsV2, rendererClasses [&NiagaraSpriteRendererProperties]}
  {
   bEnabled=false
   emitterName=FruitSplashesMergeOld
   simTarget=GPUComputeSim
   rendererClasses [&NiagaraSpriteRendererProperties]
  }
  {bEnabled=false, emitterName=Debugging, rendererClasses [&NiagaraSpriteRendererProperties]}
  {bEnabled=false, emitterName=Empty, rendererClasses []}
  {
   bEnabled=false
   emitterName=FruitSplashesAndDrops
   simTarget=GPUComputeSim
   rendererClasses [&NiagaraSpriteRendererProperties]
  }
  {emitterName=FruitSplashesAndDrops001, simTarget=GPUComputeSim, rendererClasses [&NiagaraSpriteRendererProperties]}
 ]
 userVariables [
  User.bIsFrontEnd {type=NiagaraBool}
  User.bIsFrontEndPreview {type=NiagaraBool}
  User.Grid2DScale {type=NiagaraFloat, defaultValue=1.2, description=""}
  User.WasHitByVeh {type=NiagaraBool}
 ]
}
