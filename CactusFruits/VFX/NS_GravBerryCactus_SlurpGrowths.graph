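# Niagara effect for the GravBerry cactus slurp growths, spraying short-lived cyan-teal splash sprites from sockets on a static mesh.
# Two emitters, Splashes and SplashesFruits001, fire staggered timed bursts, with each spawn group mapped to a socket index, offset by 15 in Splashes.
# Particles shoot in a cone along the socket X axis, then drag, gravity and curves slow, stretch, tint and fade them.
# Quality-level platform checks set the particle count per burst.
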NiagaraSystem /CactusFruits/VFX/NS_GravBerryCactus_SlurpGrowths
dependencies {
 usedDataInterfaces [
  &NiagaraDataInterfacePlatformSet
  &NiagaraDataInterfaceCurve
  &NiagaraDataInterfaceVector2DCurve
  &NiagaraDataInterfaceColorCurve
  &NiagaraDataInterfaceArrayFloat
  &NiagaraDataInterfaceStaticMesh
 ]
 usedDynamicInputs [&Vector2DFromCurve]
 usedModules [
  &self:SystemSpawnScript.SystemScriptSource.SystemScriptGraph.NiagaraNodeAssignment_3.SetVariables_#1
  &SystemState
  &self:SystemSpawnScript.SystemScriptSource.SystemScriptGraph.NiagaraNodeAssignment_0.SetVariables_#2
  &self:ScratchModule
 ]
}
emitters {
 Splashes {
  emitterSpawnScript [
   /SetVariables(
    Emitter.Number Of Bulbs=3
    Emitter.Particles Per Bulb:NiagaraInt32=&MakeCustomIntFromBool(
     Bool=&IsPlatformSetActive(New Platform Set=DataInterface() {propertyValues {Platforms {qualityLevelMask=-31}}})
     True INT=1
     False INT=&MakeCustomIntFromBool(
      Bool=&IsPlatformSetActive(New Platform Set=DataInterface() {propertyValues {Platforms {qualityLevelMask=-30}}})
      True INT=4
      False INT=&MakeCustomIntFromBool(
       Bool=&IsPlatformSetActive(New Platform Set=DataInterface() {propertyValues {Platforms {qualityLevelMask=-28}}})
       True INT=5
       False INT=5
      )
     )
    )
   ) {isSetParameters=true, module=SetVariables_#3}
   /SetVariables(Emitter.SpawnTimeMultiplier=1) {isSetParameters=true, module=SetVariables_#4}
  ]
  emitterUpdateScript [
   &EmitterState(Life Cycle Mode=::NewEnumerator1, Loop Behavior=::NewEnumerator1, Loop Duration=1, UseLoopDelay=true)
   &SpawnBurst_Instantaneous(Spawn Count=$Emitter.Particles Per Bulb, Spawn Time=0.2, Use Spawn Probability=false, UseLoopCountLimit=false)
   &SpawnBurst_Instantaneous(Spawn Count=$Emitter.Particles Per Bulb, Spawn Group=1, Spawn Time=0.55, Use Spawn Probability=false, UseLoopCountLimit=false) {module=SpawnBurst_Instantaneous001}
   &SpawnBurst_Instantaneous(Spawn Count=$Emitter.Particles Per Bulb, Spawn Group=2, Spawn Time=0.45, Use Spawn Probability=false, UseLoopCountLimit=false) {module=SpawnBurst_Instantaneous002}
   &SpawnBurst_Instantaneous(Spawn Count=$Emitter.Particles Per Bulb, Spawn Group=3, Spawn Time=0.475, Use Spawn Probability=false, UseLoopCountLimit=false) {enabled=false, module=SpawnBurst_Instantaneous003}
   &SpawnBurst_Instantaneous(Spawn Count=$Emitter.Particles Per Bulb, Spawn Group=4, Spawn Time=0.49, Use Spawn Probability=false, UseLoopCountLimit=false) {enabled=false, module=SpawnBurst_Instantaneous004}
   &SpawnBurst_Instantaneous(Spawn Count=$Emitter.Particles Per Bulb, Spawn Group=5, Spawn Time=0.6, Use Spawn Probability=false, UseLoopCountLimit=false) {module=SpawnBurst_Instantaneous005}
   &SpawnBurst_Instantaneous(Spawn Count=$Emitter.Particles Per Bulb, Spawn Group=6, Use Spawn Probability=false, UseLoopCountLimit=false) {enabled=false, module=SpawnBurst_Instantaneous006}
   &SpawnBurst_Instantaneous(Spawn Count=$Emitter.Particles Per Bulb, Spawn Group=7, Use Spawn Probability=false, UseLoopCountLimit=false) {enabled=false, module=SpawnBurst_Instantaneous007}
   &SpawnBurst_Instantaneous(Spawn Count=$Emitter.Particles Per Bulb, Spawn Group=8, Spawn Time=0.6, Use Spawn Probability=false, UseLoopCountLimit=false) {module=SpawnBurst_Instantaneous008}
   &SpawnBurst_Instantaneous(Spawn Count=$Emitter.Particles Per Bulb, Spawn Group=9, Spawn Time=0.6, Use Spawn Probability=false, UseLoopCountLimit=false) {module=SpawnBurst_Instantaneous009}
  ]
  particleSpawnScript [
   &InitializeParticle(
    Color Maximum=<rgba 4 4 4 1>
    Color Minimum=<rgba 0.7835 0.7835 0.7835 0.8>
    Color Mode=::NewEnumerator2
    Lifetime Max=0.25
    Lifetime Min=0.5
    Lifetime Mode=::NewEnumerator1
    Mass Max=2
    Mass Mode=::NewEnumerator3
    Sprite Rotation Angle=&Multiply_Float(A=10, B=&RandomRangeFloat(Minimum=-1, Maximum=1))
    Sprite Rotation Mode=::NewEnumerator0
    Sprite Size=&Multiply_Vector2D_ByFloat(
     Float=&MaskFloatBySpawnGroup(Masked Float=2, Passthrough Float=1)
     Vector2D=&RandomRangeVector2D(Minimum=<15 35>, Maximum=<10 60>)
    )
    Sprite Size Mode=::NewEnumerator4
    Sprite UV Mode=::NewEnumerator3
    UsePositionOffset=false
    Write Material Random=true
   )
   /Niagara/Modules/Spawn/Location/StaticMeshLocation(
    Interpolate Mesh Sampling=false
    MaskSpawnGroup=false
    Mesh Sampling Type:ENiagara_StaticSamplingMode=::NewEnumerator1
    SamplePosOffset=true
    Sampled Position Offset:Vector3f=<0 0 -7>
    Sampled Position Offset Coordinate Space:ENiagaraSkelMeshTransforms=::NewEnumerator1
    Socket Index:NiagaraInt32=&Add_Integer(A=&SpawnGroup(), B=15)
    Socket Sampling Mode:ENiagara_StaticSocketMode=::NewEnumerator2
    Static Mesh:NiagaraDataInterfaceStaticMesh=$System.Static Mesh
   )
   &CameraOffset(Camera Offset Amount=15)
   &AddVelocity(
    Cone Angle=60
    Cone Axis=&ReturnMeshOrientationAxis(Mesh Orientation Axis=::XAxis, Mesh Orientation Quat=$Particles.StaticMeshLocation.SampledRotation)
    Module.UseConeFalloff=false
    Module.UseVelDistribution=false
    Override Seed=false
    Velocity Mode=::NewEnumerator2
    Velocity Speed=250
   ) {module=AddVelocity003}
  ]
  particleUpdateScript [
   &ParticleState()
   &SpriteRotationRate(Rotation Rate=&Multiply_Float(A=10, B=&RandomRangeFloat(Minimum=-1, Maximum=1))) {enabled=false}
   &Drag(
    Drag=&FloatFromCurve(
     FloatCurve=DataInterface() {propertyValues {Curve {keys [<at 0 2 RCIM_Cubic>,<at 0.5 0>]}}}
     Scale Curve=&RandomRangeFloat(Minimum=2, Maximum=2.5)
    )
    Use Linear Drag=true
    Use Rotational Drag=true
   ) {enabled=false}
   &Drag(Drag=25, Use Linear Drag=true, Use Rotational Drag=true) {module=Drag001}
   &GravityForce()
   &SolveForcesAndVelocity(Clamp Velocity=false, Limit Acceleration=false)
   &ScaleSpriteSize(
    Non-Uniform Curve Sprite Scale=DataInterface() {propertyValues {
      XCurve {keys [
        {arriveTangent=1.8917, interpMode=RCIM_Cubic, leaveTangent=1.8917, leaveTangentWeight=0.3333, tangentMode=RCTM_User, time=0, value=2}
        {arriveTangent=-0.5754, arriveTangentWeight=0.3333, interpMode=RCIM_Cubic, leaveTangent=-0.5754, tangentMode=RCTM_User, time=1, value=5}
       ]}
      YCurve {keys [
        {arriveTangent=1.8917, interpMode=RCIM_Cubic, leaveTangent=1.8917, leaveTangentWeight=0.3333, tangentMode=RCTM_User, time=0, value=1}
        {arriveTangent=-0.5754, arriveTangentWeight=0.3333, interpMode=RCIM_Cubic, leaveTangent=-0.5754, tangentMode=RCTM_User, time=1, value=2}
       ]}
     }}
    Scale Sprite Size Mode=::NewEnumerator3
   )
   &ScaleColor(
    Scale Alpha=&FloatFromCurve(FloatCurve=DataInterface() {propertyValues {Curve {keys [<at 0.0002 0.5 RCIM_Cubic>, {arriveTangent=-2.4995, interpMode=RCIM_Cubic, leaveTangent=-2.4995, tangentMode=RCTM_User, time=1, value=-1}]}}})
    Scale RGB=&MakeVectorFromLinearColorRGB(Color=&MultiplyLinearColor(
      Linear Color=&ColorFromCurve(ColorCurve=DataInterface() {propertyValues {
         AlphaCurve {keys []}
         BlueCurve {keys [<at 0.1481 40>, {arriveTangent=-41.3655, interpMode=RCIM_Cubic, leaveTangent=-41.3655, time=0.198, value=12}, <at 0.9216 8>, <at 1 8>]}
         GreenCurve {keys [<at 0.1481 28.1816>, {arriveTangent=-26.0882, interpMode=RCIM_Cubic, leaveTangent=-26.0882, time=0.198, value=12}, <at 0.9216 8>, <at 1 8>]}
         RedCurve {keys [<at 0.1481 7.2711>, {arriveTangent=0.9422, interpMode=RCIM_Cubic, leaveTangent=0.9422, time=0.198, value=12}, <at 0.9216 8>, <at 1 8>]}
        }})
      Scale Factor=&MakeVector4_FromLinearColor(LinearColor=&LerpLinearColors(StartColor=<rgba 0.0019 0.5703 0.6979 1>, EndColor=<rgba 0.002 0.724 0.6397 1>, LerpFactor=$Particles.MaterialRandom))
     ))
    ScaleA=true
    ScaleRGB=true
   )
   &SpriteFacingAndAlignment(
    Sprite Alignment=&ReturnMeshOrientationAxis(Mesh Orientation Axis=::XAxis, Mesh Orientation Quat=$Particles.Initial.StaticMeshLocation.SampledRotation)
    WriteAlignment=true
    WriteFacing=false
   )
  ]
  rendererClasses [&NiagaraSpriteRendererProperties]
  renderers [{rendererClass=&NiagaraSpriteRendererProperties}]
 }
 SplashesFruits001 {
  emitterSpawnScript [
   /SetVariables(
    Emitter.Number Of Bulbs=3
    Emitter.Particles Per Bulb:NiagaraInt32=&MakeCustomIntFromBool(
     Bool=&IsPlatformSetActive(New Platform Set=DataInterface() {propertyValues {Platforms {qualityLevelMask=-31}}})
     True INT=1
     False INT=&MakeCustomIntFromBool(
      Bool=&IsPlatformSetActive(New Platform Set=DataInterface() {propertyValues {Platforms {qualityLevelMask=-30}}})
      True INT=1
      False INT=&MakeCustomIntFromBool(
       Bool=&IsPlatformSetActive(New Platform Set=DataInterface() {propertyValues {Platforms {qualityLevelMask=-28}}})
       True INT=1
       False INT=5
      )
     )
    )
   ) {isSetParameters=true, module=SetVariables_#3}
   /SetVariables(Emitter.SpawnTimeMultiplier:NiagaraFloat=&Multiply_Float(A=&Divide_Float(A=1, B=14), B=2.75)) {enabled=false, isSetParameters=true, module=SetVariables_#4}
   /SetVariables(Emitter.SpawnTimeMultiplier:NiagaraFloat=&Multiply_Float(A=2.75, B=0.5)) {isSetParameters=true}
   /SetVariables(Emitter.Fruit Spawn Times=DataInterface() {propertyValues {
      GpuSyncMode=SyncCpuToGpu
      MaxElements=0
      FloatData [1,0.9024,0.7979,0.7657,0.7274,0.5922,0.5639,0.491,0.4815,0.4179,0.3676,0.3024,0.2906,0.1468,0]
     }}) {isSetParameters=true}
   /SetVariables(Emitter.Burst Time low=0.5) {isSetParameters=true}
   /SetVariables(Emitter.Burst Time High=5.5) {isSetParameters=true}
  ]
  emitterUpdateScript [
   &EmitterState(Life Cycle Mode=::NewEnumerator1, Loop Behavior=::NewEnumerator1, Loop Duration=3, UseLoopDelay=true)
   &SpawnBurst_Instantaneous(Spawn Count=$Emitter.Particles Per Bulb, Spawn Time=1.85, Use Spawn Probability=false, UseLoopCountLimit=false)
   &SpawnBurst_Instantaneous(Spawn Count=$Emitter.Particles Per Bulb, Spawn Group=1, Spawn Time=2, Use Spawn Probability=false, UseLoopCountLimit=false) {module=SpawnBurst_Instantaneous001}
   &SpawnBurst_Instantaneous(Spawn Count=$Emitter.Particles Per Bulb, Spawn Group=2, Spawn Time=2.15, Use Spawn Probability=false, UseLoopCountLimit=false) {module=SpawnBurst_Instantaneous002}
   &SpawnBurst_Instantaneous(Spawn Count=$Emitter.Particles Per Bulb, Spawn Group=3, Spawn Time=2.2, Use Spawn Probability=false, UseLoopCountLimit=false) {module=SpawnBurst_Instantaneous003}
   &SpawnBurst_Instantaneous(Spawn Count=$Emitter.Particles Per Bulb, Spawn Group=4, Spawn Time=2.25, Use Spawn Probability=false, UseLoopCountLimit=false) {module=SpawnBurst_Instantaneous004}
   &SpawnBurst_Instantaneous(Spawn Count=$Emitter.Particles Per Bulb, Spawn Group=5, Spawn Time=2.4, Use Spawn Probability=false, UseLoopCountLimit=false) {module=SpawnBurst_Instantaneous005}
   &SpawnBurst_Instantaneous(Spawn Count=$Emitter.Particles Per Bulb, Spawn Group=6, Spawn Time=2.5, Use Spawn Probability=false, UseLoopCountLimit=false) {module=SpawnBurst_Instantaneous006}
   &SpawnBurst_Instantaneous(Spawn Count=$Emitter.Particles Per Bulb, Spawn Group=7, Spawn Time=2.55, Use Spawn Probability=false, UseLoopCountLimit=false) {module=SpawnBurst_Instantaneous007}
   &SpawnBurst_Instantaneous(Spawn Count=$Emitter.Particles Per Bulb, Spawn Group=8, Spawn Time=2.6, Use Spawn Probability=false, UseLoopCountLimit=false) {module=SpawnBurst_Instantaneous008}
   &SpawnBurst_Instantaneous(Spawn Count=$Emitter.Particles Per Bulb, Spawn Group=9, Spawn Time=2.65, Use Spawn Probability=false, UseLoopCountLimit=false) {module=SpawnBurst_Instantaneous009}
   &SpawnBurst_Instantaneous(Spawn Count=$Emitter.Particles Per Bulb, Spawn Group=10, Spawn Time=2.7, Use Spawn Probability=false, UseLoopCountLimit=false) {module=SpawnBurst_Instantaneous010}
   &SpawnBurst_Instantaneous(Spawn Count=$Emitter.Particles Per Bulb, Spawn Group=11, Spawn Time=1.5, Use Spawn Probability=false, UseLoopCountLimit=false) {module=SpawnBurst_Instantaneous011}
   &SpawnBurst_Instantaneous(Spawn Count=$Emitter.Particles Per Bulb, Spawn Group=12, Spawn Time=1.5, Use Spawn Probability=false, UseLoopCountLimit=false) {module=SpawnBurst_Instantaneous012}
   &SpawnBurst_Instantaneous(Spawn Count=$Emitter.Particles Per Bulb, Spawn Group=13, Spawn Time=1.6, Use Spawn Probability=false, UseLoopCountLimit=false) {module=SpawnBurst_Instantaneous013}
   &SpawnBurst_Instantaneous(Spawn Count=$Emitter.Particles Per Bulb, Spawn Group=14, Spawn Time=1.75, Use Spawn Probability=false, UseLoopCountLimit=false) {module=SpawnBurst_Instantaneous014}
  ]
  particleSpawnScript [
   &InitializeParticle(
    Color Maximum=<rgba 4 4 4 1>
    Color Minimum=<rgba 0.7835 0.7835 0.7835 0.6>
    Color Mode=::NewEnumerator2
    Lifetime Max=0.3
    Lifetime Min=0.3
    Lifetime Mode=::NewEnumerator1
    Mass Max=2
    Mass Mode=::NewEnumerator3
    Sprite Rotation Angle=&Multiply_Float(A=10, B=&RandomRangeFloat(Minimum=-1, Maximum=1))
    Sprite Rotation Mode=::NewEnumerator0
    Sprite Size=&Multiply_Vector2D_ByFloat(Float=1, Vector2D=&RandomRangeVector2D(Minimum=<5 15>, Maximum=<8 20>))
    Sprite Size Mode=::NewEnumerator4
    Sprite UV Mode=::NewEnumerator3
    UsePositionOffset=false
    Write Material Random=true
   )
   &StaticMeshLocation(
    Interpolate Mesh Sampling=false
    MaskSpawnGroup=false
    Mesh Sampling Type:ENiagara_StaticSamplingMode=::NewEnumerator1
    SamplePosOffset=true
    Sampled Position Offset:Vector3f=<0 0 -7>
    Sampled Position Offset Coordinate Space:ENiagaraSkelMeshTransforms=::NewEnumerator1
    Socket Index:NiagaraInt32=&SpawnGroup()
    Socket Sampling Mode:ENiagara_StaticSocketMode=::NewEnumerator3
    Static Mesh:NiagaraDataInterfaceStaticMesh=$System.Static Mesh
   )
   &CameraOffset(Camera Offset Amount=15)
   &AddVelocity(
    Cone Angle=60
    Cone Axis=&ReturnMeshOrientationAxis(Mesh Orientation Axis=::XAxis, Mesh Orientation Quat=$Particles.StaticMeshLocation.SampledRotation)
    Module.UseConeFalloff=false
    Module.UseVelDistribution=false
    Override Seed=false
    Velocity Mode=::NewEnumerator2
    Velocity Speed=250
   ) {module=AddVelocity003}
  ]
  particleUpdateScript [
   &ParticleState()
   &SpriteRotationRate(Rotation Rate=&Multiply_Float(A=10, B=&RandomRangeFloat(Minimum=-1, Maximum=1))) {enabled=false}
   &Drag(
    Drag=&FloatFromCurve(
     FloatCurve=DataInterface() {propertyValues {Curve {keys [<at 0 2 RCIM_Cubic>,<at 0.5 0>]}}}
     Scale Curve=&RandomRangeFloat(Minimum=2, Maximum=2.5)
    )
    Use Linear Drag=true
    Use Rotational Drag=true
   ) {enabled=false}
   &Drag(Drag=25, Use Linear Drag=true, Use Rotational Drag=true) {module=Drag001}
   &GravityForce(Gravity=<0 0 -300>)
   &SolveForcesAndVelocity(Clamp Velocity=false, Limit Acceleration=false)
   &ScaleSpriteSize(
    Non-Uniform Curve Scale=<1.2 1>
    Non-Uniform Curve Sprite Scale=DataInterface() {propertyValues {
      XCurve {keys [
        {arriveTangent=1.8917, interpMode=RCIM_Cubic, leaveTangent=1.8917, leaveTangentWeight=0.3333, tangentMode=RCTM_User, time=0, value=2}
        {arriveTangent=-0.5754, arriveTangentWeight=0.3333, interpMode=RCIM_Cubic, leaveTangent=-0.5754, tangentMode=RCTM_User, time=1, value=3}
       ]}
      YCurve {keys [
        {arriveTangent=1.8917, interpMode=RCIM_Cubic, leaveTangent=1.8917, leaveTangentWeight=0.3333, tangentMode=RCTM_User, time=0, value=1}
        {arriveTangent=-0.5754, arriveTangentWeight=0.3333, interpMode=RCIM_Cubic, leaveTangent=-0.5754, tangentMode=RCTM_User, time=1, value=2}
       ]}
     }}
    Scale Sprite Size Mode=::NewEnumerator3
   )
   &ScaleColor(
    Scale Alpha=&FloatFromCurve(FloatCurve=DataInterface() {propertyValues {Curve {keys [
         <at 0 -1>
         <at 0.0002 0.5 RCIM_Cubic>
         {arriveTangent=-2.4995, interpMode=RCIM_Cubic, leaveTangent=-2.4995, tangentMode=RCTM_User, time=1, value=-1}
        ]}}})
    Scale RGB=&MakeVectorFromLinearColorRGB(Color=&MultiplyLinearColor(
      Linear Color=&ColorFromCurve(ColorCurve=DataInterface() {propertyValues {
         AlphaCurve {keys []}
         BlueCurve {keys [<at 0.1481 40>, {arriveTangent=-41.3655, interpMode=RCIM_Cubic, leaveTangent=-41.3655, time=0.198, value=12}, <at 0.9216 8>, <at 1 8>]}
         GreenCurve {keys [<at 0.1481 28.1816>, {arriveTangent=-26.0882, interpMode=RCIM_Cubic, leaveTangent=-26.0882, time=0.198, value=12}, <at 0.9216 8>, <at 1 8>]}
         RedCurve {keys [<at 0.1481 7.2711>, {arriveTangent=0.9422, interpMode=RCIM_Cubic, leaveTangent=0.9422, time=0.198, value=12}, <at 0.9216 8>, <at 1 8>]}
        }})
      Scale Factor=&MakeVector4_FromLinearColor(LinearColor=&LerpLinearColors(StartColor=<rgba 0.0019 0.5703 0.6979 1>, EndColor=<rgba 0.002 0.724 0.6397 1>, LerpFactor=$Particles.MaterialRandom))
     ))
    ScaleA=true
    ScaleRGB=true
   )
   &SpriteFacingAndAlignment(
    Sprite Alignment=&ReturnMeshOrientationAxis(Mesh Orientation Axis=::XAxis, Mesh Orientation Quat=$Particles.Initial.StaticMeshLocation.SampledRotation)
    WriteAlignment=true
    WriteFacing=false
   )
  ]
  rendererClasses [&NiagaraSpriteRendererProperties]
  renderers [{rendererClass=&NiagaraSpriteRendererProperties}]
 }
}
summary {
 systemName=NS_GravBerryCactus_SlurpGrowths
 emitters [
  {emitterName=Splashes, rendererClasses [&NiagaraSpriteRendererProperties]}
  {emitterName=SplashesFruits001, rendererClasses [&NiagaraSpriteRendererProperties]}
 ]
 userVariables [
  User.Grid2DScale {type=NiagaraFloat, defaultValue=1.2, description=""}
  User.WasHitByVeh {type=NiagaraBool}
 ]
}
